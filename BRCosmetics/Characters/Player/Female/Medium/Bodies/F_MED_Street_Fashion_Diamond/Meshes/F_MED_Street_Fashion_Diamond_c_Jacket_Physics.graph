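# Physics asset for the jacket of the female medium Street Fashion Diamond player body, giving its dyn_jacket bones bodies and joints for secondary motion.
# Box bodies cover the jacket panels, and two capsules on spine_02 and spine_04 form the torso.
# Side and back panels attach directly to spine_02 with limited swing and tight twist.
# Each front side is a two-link chain under an angularly free MSTR body on spine_02, with the left limited tighter than the right.
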PhysicsAsset /BRCosmetics/Characters/Player/Female/Medium/Bodies/F_MED_Street_Fashion_Diamond/Meshes/F_MED_Street_Fashion_Diamond_c_Jacket_Physics
bodies {
 dyn_jacket_bk_01 [{shapeType=Box, extentX=5.9922, extentY=9.6977, extentZ=0.8796, center=<2.0915 0 0.2568>}]
 dyn_jacket_bk_l_01 [{shapeType=Box, extentX=7.1007, extentY=3.6598, extentZ=0.7705, center=<2.0481 0 -0.2323>}]
 dyn_jacket_bk_r_01 [{shapeType=Box, extentX=7.0477, extentY=3.7529, extentZ=0.7894, center=<2.0699 0 0.3634>}]
 dyn_jacket_fr_l_01 [{shapeType=Box, extentX=3.7909, extentY=2.9048, extentZ=0.973, center=<1.4288 0 0>}]
 dyn_jacket_fr_l_02 [{shapeType=Box, extentX=4.4106, extentY=3.6201, extentZ=0.7988, center=<0.8638 -0.0765 -0.03>}]
 dyn_jacket_fr_l_MSTR [{shapeType=Box, extentX=1.01, extentY=1.01, extentZ=1.01}]
 dyn_jacket_fr_r_01 [{shapeType=Box, extentX=4.0883, extentY=1.5819, extentZ=1.01, center=<1.2707 0 0>}]
 dyn_jacket_fr_r_02 [{shapeType=Box, extentX=2.3682, extentY=2.1111, extentZ=1.01, center=<0.7312 0 0>}]
 dyn_jacket_fr_r_MSTR [{shapeType=Box, extentX=1.01, extentY=1.01, extentZ=1.01, rotation=<rot 0 -90 0>}]
 dyn_jacket_l_01 [{shapeType=Box, extentX=7.2501, extentY=11.5506, extentZ=1.01, center=<3.4431 0.0525 0.2093>}]
 dyn_jacket_r_01 [{shapeType=Box, extentX=7.2096, extentY=11.3904, extentZ=1.01, center=<3.429 -0.0667 0.3592>}]
 spine_02 [{shapeType=Capsule, length=3.6734, radius=7.5918, center=<2.39 2.3015 0.2245>}]
 spine_04 [{shapeType=Capsule, length=2.65, radius=8.9675, center=<-1.8917 2.2613 -0.5047>, rotation=<rot 4.0259 -90 -0.603>}]
}
constraints [
 {bone1Name=dyn_jacket_fr_r_02, bone2Name=dyn_jacket_fr_r_01, swing1LimitDegrees=10, swing2LimitDegrees=45, twistLimitDegrees=2}
 {bone1Name=dyn_jacket_fr_l_02, bone2Name=dyn_jacket_fr_l_01, swing1LimitDegrees=1, swing2LimitDegrees=15, twistLimitDegrees=1}
 {bone1Name=dyn_jacket_l_01, bone2Name=spine_02, swing1LimitDegrees=10, swing2LimitDegrees=45, twistLimitDegrees=2}
 {bone1Name=dyn_jacket_r_01, bone2Name=spine_02, swing1LimitDegrees=10, swing2LimitDegrees=45, twistLimitDegrees=2}
 {bone1Name=dyn_jacket_bk_l_01, bone2Name=spine_02, swing1LimitDegrees=10, swing2LimitDegrees=45, twistLimitDegrees=2}
 {bone1Name=dyn_jacket_bk_r_01, bone2Name=spine_02, swing1LimitDegrees=10, swing2LimitDegrees=45, twistLimitDegrees=2}
 {bone1Name=dyn_jacket_bk_01, bone2Name=spine_02, swing1LimitDegrees=10, swing2LimitDegrees=45, twistLimitDegrees=2}
 {bone1Name=dyn_jacket_fr_r_01, bone2Name=dyn_jacket_fr_r_MSTR, swing1LimitDegrees=10, swing2LimitDegrees=45, twistLimitDegrees=2}
 {bone1Name=dyn_jacket_fr_l_01, bone2Name=dyn_jacket_fr_l_MSTR, swing1LimitDegrees=1, swing2LimitDegrees=15, twistLimitDegrees=1}
 {
  bone1Name=spine_04
  bone2Name=spine_02
  swing1LimitDegrees=45
  swing1Motion=Free
  swing2LimitDegrees=45
  swing2Motion=Free
  twistLimitDegrees=45
  twistMotion=Free
 }
 {
  bone1Name=dyn_jacket_fr_l_MSTR
  bone2Name=spine_02
  swing1LimitDegrees=45
  swing1Motion=Free
  swing2LimitDegrees=45
  swing2Motion=Free
  twistLimitDegrees=45
  twistMotion=Free
 }
 {
  bone1Name=dyn_jacket_fr_r_MSTR
  bone2Name=spine_02
  swing1LimitDegrees=45
  swing1Motion=Free
  swing2LimitDegrees=45
  swing2Motion=Free
  twistLimitDegrees=45
  twistMotion=Free
 }
]
