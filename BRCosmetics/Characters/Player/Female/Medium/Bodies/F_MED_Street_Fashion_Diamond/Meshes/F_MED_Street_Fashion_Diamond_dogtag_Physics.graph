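# Physics asset for the dogtag necklace on the female medium Street Fashion Diamond player body, giving the chain and two tags physical motion.
# Five chain bodies, mostly flat boxes, link from a master body constrained to spine_05, and both tags branch from the last link.
# The top joint swings widely, then the tight limits below it loosen progressively toward the tip, favoring one swing axis.
# Capsules on spine_04 and spine_05 approximate the torso, with free-rotation constraints.
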PhysicsAsset /BRCosmetics/Characters/Player/Female/Medium/Bodies/F_MED_Street_Fashion_Diamond/Meshes/F_MED_Street_Fashion_Diamond_dogtag_Physics
bodies {
 dyn_dogtag_chain_01 [{shapeType=Capsule, length=5.5012, radius=2.2634, center=<-0.2933 0 0>, rotation=<rot 0 0 90>}]
 dyn_dogtag_chain_02 [{shapeType=Box, extentX=5, extentY=10, extentZ=1.3127, center=<0.816 -0.0105 -0.0136>, rotation=<rot -0.0033 0.4506 0.4098>}]
 dyn_dogtag_chain_03 [{shapeType=Box, extentX=2.8863, extentY=8.5, extentZ=1.2183, center=<1.2084 0.0073 0>, rotation=<rot -0.0233 1.3802 0.9727>}]
 dyn_dogtag_chain_04 [{shapeType=Box, extentX=2.9668, extentY=6, extentZ=1.2038, center=<1.3169 -0.3822 0.005>, rotation=<rot -0.0112 0.8449 0.7606>}]
 dyn_dogtag_chain_05 [{shapeType=Box, extentX=2.9384, extentY=4.8535, extentZ=1.1945, center=<1.4293 -0.1887 0>}]
 dyn_dogtag_chain_MSTR [{shapeType=Box, extentX=1.01, extentY=1.01, extentZ=1.01}]
 dyn_dogtag_l [{shapeType=Box, extentX=5.5408, extentY=2.9519, extentZ=0.9712, center=<1.4639 -0.3373 -0.1896>, rotation=<rot -8.3584 0 0>}]
 dyn_dogtag_r [{shapeType=Box, extentX=5.1766, extentY=3.1477, extentZ=1.0021, center=<1.5011 0.1485 -0.3008>, rotation=<rot -8.1853 0 0>}]
 spine_04 [
  {shapeType=Capsule, length=14.517, radius=6.2689, center=<4.549 4.0352 -0.3739>, rotation=<rot 0 -1.1713 0>}
  {shapeType=Capsule, length=12.3417, radius=5.2396, center=<1.8101 6.7601 -0.3739>}
 ]
 spine_05 [{shapeType=Capsule, length=11.2958, radius=10.3543, center=<-2.191 -1.2104 -0.3049>, rotation=<rot 0 -10.891 0>}]
}
constraints [
 {bone1Name=dyn_dogtag_chain_01, bone2Name=dyn_dogtag_chain_MSTR, swing1LimitDegrees=45, swing2LimitDegrees=45, twistLimitDegrees=15}
 {bone1Name=dyn_dogtag_chain_05, bone2Name=dyn_dogtag_chain_04, swing1LimitDegrees=10, swing2LimitDegrees=30, twistLimitDegrees=3}
 {bone1Name=dyn_dogtag_chain_04, bone2Name=dyn_dogtag_chain_03, swing1LimitDegrees=7, swing2LimitDegrees=25, twistLimitDegrees=3}
 {bone1Name=dyn_dogtag_chain_03, bone2Name=dyn_dogtag_chain_02, swing1LimitDegrees=5, swing2LimitDegrees=20, twistLimitDegrees=3}
 {bone1Name=dyn_dogtag_chain_02, bone2Name=dyn_dogtag_chain_01, swing1LimitDegrees=3, swing2LimitDegrees=15, twistLimitDegrees=1}
 {bone1Name=dyn_dogtag_l, bone2Name=dyn_dogtag_chain_05, swing1LimitDegrees=3, swing2LimitDegrees=30, twistLimitDegrees=3}
 {bone1Name=dyn_dogtag_r, bone2Name=dyn_dogtag_chain_05, swing1LimitDegrees=3, swing2LimitDegrees=30, twistLimitDegrees=3}
 {
  bone1Name=dyn_dogtag_chain_MSTR
  bone2Name=spine_05
  swing1LimitDegrees=45
  swing1Motion=Free
  swing2LimitDegrees=45
  swing2Motion=Free
  twistLimitDegrees=45
  twistMotion=Free
 }
 {
  bone1Name=spine_05
  bone2Name=spine_04
  swing1LimitDegrees=45
  swing1Motion=Free
  swing2LimitDegrees=45
  swing2Motion=Free
  twistLimitDegrees=45
  twistMotion=Free
 }
]
