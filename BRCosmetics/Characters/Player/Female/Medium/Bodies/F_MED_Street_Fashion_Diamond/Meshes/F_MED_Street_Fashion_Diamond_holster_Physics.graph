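# Physics asset for the holster mesh of the female medium Street Fashion Diamond player body, setting up jointed dangling straps.
# Front and back chains of box bodies, each including a buckle, hang from MSTR roots constrained to spine_02.
# Strap joints nearly lock twist and swing mainly on one axis, while root and body joints rotate freely.
# Capsules on pelvis, lower spine and right thigh approximate the body beneath the straps.
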PhysicsAsset /BRCosmetics/Characters/Player/Female/Medium/Bodies/F_MED_Street_Fashion_Diamond/Meshes/F_MED_Street_Fashion_Diamond_holster_Physics
bodies {
 dyn_holster_strap_bk_01 [{shapeType=Box, extentX=3.9363, extentY=4.2095, extentZ=1.3359, center=<1.4447 0 0.225>, rotation=<rot 0 -89.9999 -5.0002>}]
 dyn_holster_strap_bk_03 [{shapeType=Box, extentX=8.4362, extentY=3.4735, extentZ=1.2535, center=<3.9179 0.0221 0.4519>}]
 dyn_holster_strap_bk_MSTR [{shapeType=Box, extentX=1.01, extentY=1.01, extentZ=1.01, rotation=<rot 0 -90 0>}]
 dyn_holster_strap_bk_buckle_01 [{shapeType=Box, extentX=3.8293, extentY=1.1883, extentZ=4.7085, center=<1.2863 0.0248 0.1755>, rotation=<rot 0 1.4099 -76.8943>}]
 dyn_holster_strap_fr_01 [{shapeType=Box, extentX=4.5604, extentY=3.303, extentZ=1.5004, center=<0.8782 0 0>}]
 dyn_holster_strap_fr_03 [{shapeType=Box, extentX=5.9861, extentY=3.4948, extentZ=1.8483, center=<2.035 0.0264 0.0259>}]
 dyn_holster_strap_fr_04 [{shapeType=Box, extentX=6.3292, extentY=1.2956, extentZ=3.5182, center=<2.0179 -0.006 0.0017>, rotation=<rot 0 178.9725 -85.9652>}]
 dyn_holster_strap_fr_MSTR [{shapeType=Box, extentX=1.01, extentY=1.01, extentZ=1.01}]
 dyn_holster_strap_fr_buckle_01 [{shapeType=Box, extentX=4.0238, extentY=4.9544, extentZ=1.6924, center=<1.4753 -0.0413 0.1144>}]
 pelvis [
  {shapeType=Capsule, length=9.3102, radius=9.0653, center=<0.2799 -1.1379 0.0412>}
  {shapeType=Capsule, length=10.1089, radius=8.9936, center=<-3.0952 -3.8967 0>}
 ]
 spine_01 [
  {shapeType=Capsule, length=7.7314, radius=7.5079, rotation=<rot 0 -90 0>}
  {shapeType=Capsule, length=7.1152, radius=7.1184, center=<-1.2243 1.308 0>}
 ]
 spine_02 [{shapeType=Capsule, length=4.8714, radius=7.0605, center=<2.39 1.6479 0.2245>}]
 thigh_r [{shapeType=Capsule, length=18.84, radius=7.7986, center=<9.8063 -0.0104 -1.3042>, rotation=<rot 84.9942 174.9993 180>}]
}
constraints [
 {bone1Name=dyn_holster_strap_bk_01, bone2Name=dyn_holster_strap_bk_MSTR, swing1LimitDegrees=3, swing2LimitDegrees=15, twistLimitDegrees=1}
 {
  bone1Name=dyn_holster_strap_bk_MSTR
  bone2Name=spine_02
  swing1LimitDegrees=45
  swing1Motion=Free
  swing2LimitDegrees=45
  swing2Motion=Free
  twistLimitDegrees=45
  twistMotion=Free
 }
 {
  bone1Name=spine_01
  bone2Name=pelvis
  swing1LimitDegrees=45
  swing1Motion=Free
  swing2LimitDegrees=45
  swing2Motion=Free
  twistLimitDegrees=45
  twistMotion=Free
 }
 {
  bone1Name=spine_02
  bone2Name=spine_01
  swing1LimitDegrees=45
  swing1Motion=Free
  swing2LimitDegrees=45
  swing2Motion=Free
  twistLimitDegrees=45
  twistMotion=Free
 }
 {
  bone1Name=dyn_holster_strap_bk_buckle_01
  bone2Name=dyn_holster_strap_bk_01
  swing1LimitDegrees=5
  swing2LimitDegrees=25
  twistLimitDegrees=1
 }
 {
  bone1Name=dyn_holster_strap_bk_03
  bone2Name=dyn_holster_strap_bk_buckle_01
  swing1LimitDegrees=1
  swing2LimitDegrees=25
  twistLimitDegrees=1
 }
 {
  bone1Name=dyn_holster_strap_fr_MSTR
  bone2Name=spine_02
  swing1LimitDegrees=45
  swing1Motion=Free
  swing2LimitDegrees=45
  swing2Motion=Free
  twistLimitDegrees=45
  twistMotion=Free
 }
 {bone1Name=dyn_holster_strap_fr_01, bone2Name=dyn_holster_strap_fr_MSTR, swing1LimitDegrees=3, swing2LimitDegrees=15, twistLimitDegrees=1}
 {
  bone1Name=dyn_holster_strap_fr_buckle_01
  bone2Name=dyn_holster_strap_fr_01
  swing1LimitDegrees=3
  swing2LimitDegrees=25
  twistLimitDegrees=1
 }
 {
  bone1Name=dyn_holster_strap_fr_03
  bone2Name=dyn_holster_strap_fr_buckle_01
  swing1LimitDegrees=3
  swing2LimitDegrees=25
  twistLimitDegrees=1
 }
 {bone1Name=dyn_holster_strap_fr_04, bone2Name=dyn_holster_strap_fr_03, swing1LimitDegrees=3, swing2LimitDegrees=25, twistLimitDegrees=1}
 {
  bone1Name=thigh_r
  bone2Name=pelvis
  swing1LimitDegrees=45
  swing1Motion=Free
  swing2LimitDegrees=45
  swing2Motion=Free
  twistLimitDegrees=45
  twistMotion=Free
 }
]
