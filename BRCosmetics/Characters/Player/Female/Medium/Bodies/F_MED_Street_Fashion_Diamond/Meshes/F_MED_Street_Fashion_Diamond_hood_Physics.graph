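# Physics asset for the hood of the female medium Street Fashion Diamond body cosmetic, covering its left and right dynamic hood bones.
# Each hood bone has a thin box body, and spine_05 has a capsule representing the upper torso.
# Two constraints tie the hood bodies to spine_05 with identical swing and twist limit values, with no constraint between the two halves.
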PhysicsAsset /BRCosmetics/Characters/Player/Female/Medium/Bodies/F_MED_Street_Fashion_Diamond/Meshes/F_MED_Street_Fashion_Diamond_hood_Physics
bodies {
 dyn_hood_l [{shapeType=Box, extentX=9.8438, extentY=8.7804, extentZ=1.8654, center=<-0.9048 -1.1153 -0.0128>}]
 dyn_hood_r [{shapeType=Box, extentX=9.7899, extentY=8.7363, extentZ=1.836, center=<-0.8153 0.9258 0.087>}]
 spine_05 [{shapeType=Capsule, length=12.707, radius=10.1487, center=<1.3889 0.3206 -0.1711>, rotation=<rot 1.3848 -90 -1.5254>}]
}
constraints [
 {bone1Name=dyn_hood_l, bone2Name=spine_05, swing1LimitDegrees=35, swing2LimitDegrees=60, twistLimitDegrees=10}
 {bone1Name=dyn_hood_r, bone2Name=spine_05, swing1LimitDegrees=35, swing2LimitDegrees=60, twistLimitDegrees=10}
]
