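# Physics asset for the dangling accessories on the female medium Street Fashion Diamond body: right thigh pouch straps, right arm straps, belt knot and ropes, and left calf knife.
# Limb and pelvis shapes provide body volumes around small dyn_ bone shapes.
# Each accessory hangs from a MSTR bone through tightly twist-limited, swing-limited joints.
# Fully free-rotation constraints link MSTR and body bones; the arm strap master links to pelvis, not upper arm.
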PhysicsAsset /BRCosmetics/Characters/Player/Female/Medium/Bodies/F_MED_Street_Fashion_Diamond/Meshes/F_MED_Street_Fashion_Diamond_pouch_physics
bodies {
 calf_twist_02_l [{shapeType=Capsule, length=10.0489, radius=7.1449, center=<-4.133 -1.6764 -0.8564>, rotation=<rot 90 0 0>}]
 dyn_arm_r_strap_01 [{shapeType=Box, extentX=2.4945, extentY=1.01, extentZ=1.6032, center=<0.64 0.0527 -0.0076>, rotation=<rot 0.5049 -179.1601 -84.2195>}]
 dyn_arm_r_strap_02 [{shapeType=Box, extentX=1.7867, extentY=0.7383, extentZ=3.1857, center=<1.286 0.1091 -0.0381>, rotation=<rot 0.0262 90.6259 -89.701>}]
 dyn_arm_r_strap_MSTR [{shapeType=Box, extentX=1.01, extentY=1.01, extentZ=1.01, rotation=<rot 0 -90 0>}]
 dyn_belt_knot_MSTR [{shapeType=Box, extentX=1.01, extentY=1.01, extentZ=1.01}]
 dyn_belt_knot_back [{shapeType=Box, extentX=2.3773, extentY=2.8154, extentZ=0.8385, center=<1.7851 0.2357 0.0099>}]
 dyn_belt_knot_front [{shapeType=Box, extentX=4.4248, extentY=4.2122, extentZ=1.083, center=<3.5933 -0.1873 0.0329>}]
 dyn_belt_rope_01 [{shapeType=Capsule, length=1.9115, radius=0.505, center=<1.9937 0.0133 0.2547>, rotation=<rot 0 90.6648 -82.6292>}]
 dyn_belt_rope_02 [{shapeType=Capsule, length=2.1771, radius=0.5567, center=<1.2813 -0.0114 -0.0101>, rotation=<rot 0 89.9249 -86.7232>}]
 dyn_knife [{shapeType=Box, extentX=9.0753, extentY=2.6676, extentZ=30.9533, center=<1.1223 -0.3674 0.8998>, rotation=<rot -2.977 -90 2.2431>}]
 dyn_knife_MSTR [{shapeType=Box, extentX=1.01, extentY=1.01, extentZ=1.01}]
 dyn_thigh_r_pouch_strap_01 [{shapeType=Box, extentX=2.3512, extentY=2.2092, extentZ=0.6043, center=<1.1585 -0.0966 0.0209>}]
 dyn_thigh_r_pouch_strap_02 [{shapeType=Box, extentX=3.185, extentY=1.9038, extentZ=0.5, center=<1.5714 -0.0107 -0.0501>}]
 dyn_thigh_r_pouch_strap_03 [{shapeType=Box, extentX=2.0235, extentY=1.8473, extentZ=0.5, center=<1.485 0.027 -0.0735>}]
 dyn_thigh_r_pouch_strap_MSTR [{shapeType=Box, extentX=1.01, extentY=1.01, extentZ=1.01, rotation=<rot 0 -90 0>}]
 pelvis [{shapeType=Capsule, length=8.8825, radius=9.3786, center=<0.3156 -1.7644 0.2056>}]
 thigh_twist_01_r [{shapeType=Box, extentX=5.8958, extentY=5.5131, extentZ=13.9955, center=<5.2017 -2.3549 4.1496>, rotation=<rot -13.1426 -89.9995 -85.1909>}]
 upperarm_twist_02_r [{shapeType=Capsule, length=6.1224, radius=4.241, center=<1.14 0.9034 0.3155>, rotation=<rot 84.988 180 -180>}]
}
constraints [
 {bone1Name=dyn_thigh_r_pouch_strap_02, bone2Name=dyn_thigh_r_pouch_strap_01, swing1LimitDegrees=3, swing2LimitDegrees=25, twistLimitDegrees=1}
 {
  bone1Name=dyn_thigh_r_pouch_strap_01
  bone2Name=dyn_thigh_r_pouch_strap_MSTR
  swing1LimitDegrees=5
  swing2LimitDegrees=25
  twistLimitDegrees=1
 }
 {bone1Name=dyn_belt_rope_01, bone2Name=dyn_belt_knot_MSTR, swing1LimitDegrees=45, swing2LimitDegrees=45, twistLimitDegrees=1}
 {bone1Name=dyn_belt_rope_02, bone2Name=dyn_belt_rope_01, swing1LimitDegrees=45, swing2LimitDegrees=45, twistLimitDegrees=1}
 {bone1Name=dyn_belt_knot_back, bone2Name=dyn_belt_knot_MSTR, swing1LimitDegrees=25, swing2LimitDegrees=45, twistLimitDegrees=1}
 {bone1Name=dyn_belt_knot_front, bone2Name=dyn_belt_knot_MSTR, swing1LimitDegrees=20, swing2LimitDegrees=45, twistLimitDegrees=1}
 {bone1Name=dyn_arm_r_strap_01, bone2Name=dyn_arm_r_strap_MSTR, swing1LimitDegrees=5, swing2LimitDegrees=30, twistLimitDegrees=1}
 {bone1Name=dyn_arm_r_strap_02, bone2Name=dyn_arm_r_strap_01, swing1LimitDegrees=10, swing2LimitDegrees=40, twistLimitDegrees=1}
 {bone1Name=dyn_thigh_r_pouch_strap_03, bone2Name=dyn_thigh_r_pouch_strap_02, swing1LimitDegrees=2, swing2LimitDegrees=45, twistLimitDegrees=1}
 {
  bone1Name=dyn_thigh_r_pouch_strap_MSTR
  bone2Name=thigh_twist_01_r
  swing1LimitDegrees=45
  swing1Motion=Free
  swing2LimitDegrees=45
  swing2Motion=Free
  twistLimitDegrees=45
  twistMotion=Free
 }
 {
  bone1Name=dyn_arm_r_strap_MSTR
  bone2Name=pelvis
  swing1LimitDegrees=45
  swing1Motion=Free
  swing2LimitDegrees=45
  swing2Motion=Free
  twistLimitDegrees=45
  twistMotion=Free
 }
 {
  bone1Name=upperarm_twist_02_r
  bone2Name=pelvis
  swing1LimitDegrees=45
  swing1Motion=Free
  swing2LimitDegrees=45
  swing2Motion=Free
  twistLimitDegrees=45
  twistMotion=Free
 }
 {
  bone1Name=thigh_twist_01_r
  bone2Name=pelvis
  swing1LimitDegrees=45
  swing1Motion=Free
  swing2LimitDegrees=45
  swing2Motion=Free
  twistLimitDegrees=45
  twistMotion=Free
 }
 {
  bone1Name=dyn_belt_knot_MSTR
  bone2Name=pelvis
  swing1LimitDegrees=45
  swing1Motion=Free
  swing2LimitDegrees=45
  swing2Motion=Free
  twistLimitDegrees=45
  twistMotion=Free
 }
 {
  bone1Name=dyn_knife_MSTR
  bone2Name=calf_twist_02_l
  swing1LimitDegrees=45
  swing1Motion=Free
  swing2LimitDegrees=45
  swing2Motion=Free
  twistLimitDegrees=45
  twistMotion=Free
 }
 {
  bone1Name=calf_twist_02_l
  bone2Name=pelvis
  swing1LimitDegrees=45
  swing1Motion=Free
  swing2LimitDegrees=45
  swing2Motion=Free
  twistLimitDegrees=45
  twistMotion=Free
 }
 {bone1Name=dyn_knife, bone2Name=dyn_knife_MSTR, swing1LimitDegrees=5, swing2LimitDegrees=1, twistLimitDegrees=3}
]
