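# Post-copy secondary-motion AnimBlueprint for the F_MED_Street_Fashion_Diamond female body cosmetic, animating jacket, straps, dogtags and accessories.
# Both branches copy the attached parent's pose, offset the left glute when crouching, and constrain holster and jacket bones.
# Front end adds AnimDynamics dogtags and RigidBody jacket, pouch, holster and hood.
# Otherwise, one jacket RigidBody feeds per-bone AnimDynamics; a dogtag RigidBody is unconnected.
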AnimBlueprint /BRCosmetics/Characters/Player/Female/Medium/Bodies/F_MED_Street_Fashion_Diamond/Meshes/F_MED_Street_Fashion_Diamond_AnimBP
target_skeleton=&/F_MED_Street_Fashion_Diamond_Skeleton
graphs [
 AnimGraph {nodes [
   RigidBody() {
    bClampLinearTranslationLimitToRefPose=true
    bOverrideWorldGravity=true
    overridePhysicsAsset=&/F_MED_Street_Fashion_Diamond_dogtag_Physics
    simulationSpace=BaseBoneSpace
    alphaBoolBlend {blendOption=Linear}
    baseBoneRef {boneName=root}
    componentLinearAccScale=<0.3 0.3 0.4>
    componentLinearVelScale=<0.3 0.3 0.3>
    overrideWorldGravity {z=-980}
   }
   n1 = "Transform (Modify) Bone - Bone: dyn_glute_l"!(
    ComponentPose=Local To Component(LocalPose=Copy Pose From Mesh() {bUseAttachedParent=true})
    bAlphaBoolEnabled:bool=Get bIsCrouching()
    Translation="0.000,-2,-4"
    Rotation="0,0,0"
    Scale="1,1,1"
   ) -> (Pose) {
    alphaInputType=Bool
    translationMode=BMM_Additive
    alphaBoolBlend {blendOption=Linear}
    boneToModify {boneName=dyn_glute_l}
    translation {y=-2, z=-4}
   }
   n2 = "Constraint - dyn_holster_strap_fr_MSTR (PPPPP)"!(
    ComponentPose=@n1
    Alpha=1
    ConstraintWeights_0=0.207
    ConstraintWeights_1:float=0.3785
    ConstraintWeights_2:float=0.0522
    ConstraintWeights_3:float=0.1464
    ConstraintWeights_4:float=0.2159
   ) -> (Pose) {
    alphaBoolBlend {blendOption=Linear}
    boneToModify {boneName=dyn_holster_strap_fr_MSTR}
    constraintSetup [
     {targetBone {boneName=spine_01}}
     {targetBone {boneName=spine_02}}
     {targetBone {boneName=spine_03}}
     {targetBone {boneName=spine_04}}
     {targetBone {boneName=pelvis}}
    ]
    constraintWeights [1,1,1,1,1]
   }
   n3 = "Constraint - dyn_holster_strap_bk_MSTR (PPPPP)"!(
    ComponentPose=@n2
    Alpha=1
    ConstraintWeights_0=0.2524
    ConstraintWeights_1:float=0.3941
    ConstraintWeights_2:float=0.0639
    ConstraintWeights_3:float=0.1081
    ConstraintWeights_4:float=0.1815
   ) -> (Pose) {
    alphaBoolBlend {blendOption=Linear}
    boneToModify {boneName=dyn_holster_strap_bk_MSTR}
    constraintSetup [
     {targetBone {boneName=spine_01}}
     {targetBone {boneName=spine_02}}
     {targetBone {boneName=spine_03}}
     {targetBone {boneName=spine_04}}
     {targetBone {boneName=pelvis}}
    ]
    constraintWeights [1,1,1,1,1]
   }
   n4 = "Constraint - dyn_holster_strap_fr_buckle_02 (R)"(ComponentPose=@n3, Alpha=1, ConstraintWeights_0=1) {
    alphaBoolBlend {blendOption=Linear}
    boneToModify {boneName=dyn_holster_strap_fr_buckle_02}
    constraintSetup [{transformType=Rotation, targetBone {boneName=thigh_r}}]
    constraintWeights [1]
   }
   n5 = "Constraint - dyn_jacket_fr_r_MSTR (PPPPPPPP)"!(
    ComponentPose=@n4
    Alpha=1
    ConstraintWeights_0=0.0005
    ConstraintWeights_1:float=0.0103
    ConstraintWeights_2:float=0.0542
    ConstraintWeights_3:float=0.0584
    ConstraintWeights_4:float=0.0025
    ConstraintWeights_5:float=0.0111
    ConstraintWeights_6:float=0.8612
    ConstraintWeights_7:float=0.0018
   ) -> (Pose) {
    alphaBoolBlend {blendOption=Linear}
    boneToModify {boneName=dyn_jacket_fr_r_MSTR}
    constraintSetup [
     {targetBone {boneName=spine_01}}
     {targetBone {boneName=dyn_jacket_r_01}}
     {targetBone {boneName=dyn_jacket_fr_r_01}}
     {targetBone {boneName=dyn_jacket_fr_r_02}}
     {targetBone {boneName=spine_02}}
     {targetBone {boneName=spine_03}}
     {targetBone {boneName=spine_04}}
     {targetBone {boneName=spine_05}}
    ]
    constraintWeights [1,1,1,1,1,1,1,1]
   }
   n6 = "Constraint - dyn_jacket_fr_l_MSTR (PPPPPPPP)"!(
    ComponentPose=@n5
    Alpha=1
    ConstraintWeights_0=0.0005
    ConstraintWeights_1:float=0
    ConstraintWeights_2:float=0.1775
    ConstraintWeights_3:float=0
    ConstraintWeights_4:float=0.0001
    ConstraintWeights_5:float=0.018
    ConstraintWeights_6:float=0.8025
    ConstraintWeights_7:float=0.0014
   ) -> (Pose) {
    alphaBoolBlend {blendOption=Linear}
    boneToModify {boneName=dyn_jacket_fr_l_MSTR}
    constraintSetup [
     {targetBone {boneName=spine_03}}
     {targetBone {boneName=clavicle_l}}
     {targetBone {boneName=dyn_jacket_fr_l_01}}
     {targetBone {boneName=spine_01}}
     {targetBone {boneName=spine_02}}
     {targetBone {boneName=dyn_jacket_fr_l_02}}
     {targetBone {boneName=spine_04}}
     {targetBone {boneName=spine_05}}
    ]
    constraintWeights [1,1,1,1,1,1,1,1]
   }
   n7 = RigidBody(ComponentPose=@n6) {
    bClampLinearTranslationLimitToRefPose=true
    bOverrideWorldGravity=true
    overridePhysicsAsset=&/F_MED_Street_Fashion_Diamond_c_Jacket_Physics
    simulationSpace=BaseBoneSpace
    alphaBoolBlend {blendOption=Linear}
    baseBoneRef {boneName=pelvis}
    componentLinearAccScale=<0.5 0.5 0.7>
    componentLinearVelScale=<0.3 0.3 0.5>
    overrideWorldGravity {z=-980}
   }
   n8 = Anim Dynamics - Bone: dyn_zipper(ComponentPose=@n7, Alpha=1) {
    angularDampingOverride=0.999
    bGravityOverrideInSimSpace=true
    bOverrideAngularDamping=true
    bUseGravityOverride=true
    alphaBoolBlend {blendOption=Linear}
    boundBone {boneName=dyn_zipper}
    componentAppliedLinearAccClamp=<3000 3000 3000>
    componentLinearAccScale=<0.3 0.3 0.3>
    componentLinearVelScale=<0.3 0.3 0.3>
    gravityOverride {z=-980}
    physicsBodyDefinitions [{
      sphereCollisionRadius=0
      boundBone {boneName=dyn_zipper}
      boxExtents=<10 15 10>
      constraintSetup {angularLimitsMax=<0 5 0>, angularLimitsMin=<0 -45 0>, angularTarget=<0 0 0>}
      localJointOffset=<3 0 0>
     }]
   }
   n9 = Anim Dynamics - Bone: dyn_belt_rope_01(ComponentPose=@n8, Alpha=1, ExternalForce="0,0,0") {
    angularDampingOverride=0.998
    bGravityOverrideInSimSpace=true
    bOverrideAngularDamping=true
    bUseGravityOverride=true
    simulationSpace=RootRelative
    alphaBoolBlend {blendOption=Linear}
    boundBone {boneName=dyn_belt_rope_01}
    componentAppliedLinearAccClamp=<3000 3000 3000>
    componentLinearAccScale=<0.3 0.3 0.3>
    componentLinearVelScale=<0.3 0.3 0.3>
    gravityOverride {z=-980}
    physicsBodyDefinitions [{
      sphereCollisionRadius=0
      boundBone {boneName=dyn_belt_rope_01}
      boxExtents=<10 5 5>
      constraintSetup {angularLimitsMax=<15 0 30>, angularLimitsMin=<-15 -45 -10>, angularTarget=<0 0 0>}
      localJointOffset=<2 0 0>
     }]
    relativeSpaceBone {boneName=pelvis}
   }
   n10 = Anim Dynamics - Bone: dyn_belt_rope_02(ComponentPose=@n9, Alpha=1, ExternalForce="0,0,0") {
    angularDampingOverride=0.998
    bGravityOverrideInSimSpace=true
    bOverrideAngularDamping=true
    bUseGravityOverride=true
    simulationSpace=RootRelative
    alphaBoolBlend {blendOption=Linear}
    boundBone {boneName=dyn_belt_rope_02}
    componentAppliedLinearAccClamp=<3000 3000 3000>
    componentLinearAccScale=<0.3 0.3 0.3>
    componentLinearVelScale=<0.3 0.3 0.3>
    gravityOverride {z=-980}
    physicsBodyDefinitions [{
      sphereCollisionRadius=0
      boundBone {boneName=dyn_belt_rope_02}
      boxExtents=<10 5 5>
      constraintSetup {angularLimitsMax=<15 0 35>, angularLimitsMin=<-15 -45 -15>, angularTarget=<0 0 0>}
      localJointOffset=<2 0 0>
     }]
    relativeSpaceBone {boneName=pelvis}
   }
   n11 = Anim Dynamics - Bone: dyn_belt_knot_back(ComponentPose=@n10, Alpha=1, ExternalForce="0,0,0") {
    angularDampingOverride=0.998
    bGravityOverrideInSimSpace=true
    bOverrideAngularDamping=true
    bUseGravityOverride=true
    simulationSpace=RootRelative
    alphaBoolBlend {blendOption=Linear}
    boundBone {boneName=dyn_belt_knot_back}
    componentAppliedLinearAccClamp=<3000 3000 3000>
    componentLinearAccScale=<0.3 0.3 0.3>
    componentLinearVelScale=<0.3 0.3 0.3>
    gravityOverride {z=-980}
    physicsBodyDefinitions [{
      sphereCollisionRadius=0
      boundBone {boneName=dyn_belt_knot_back}
      boxExtents=<10 5 5>
      constraintSetup {angularLimitsMax=<15 0 45>, angularLimitsMin=<-15 -45 -45>, angularTarget=<0 0 0>}
      localJointOffset=<2 0 0>
     }]
    relativeSpaceBone {boneName=pelvis}
   }
   n12 = Anim Dynamics - Bone: dyn_belt_knot_front(ComponentPose=@n11, Alpha=1, ExternalForce="0,0,0") {
    angularDampingOverride=0.998
    bGravityOverrideInSimSpace=true
    bOverrideAngularDamping=true
    bUseGravityOverride=true
    simulationSpace=RootRelative
    alphaBoolBlend {blendOption=Linear}
    boundBone {boneName=dyn_belt_knot_front}
    componentAppliedLinearAccClamp=<3000 3000 3000>
    componentLinearAccScale=<0.3 0.3 0.3>
    componentLinearVelScale=<0.3 0.3 0.3>
    gravityOverride {z=-980}
    physicsBodyDefinitions [{
      sphereCollisionRadius=0
      boundBone {boneName=dyn_belt_knot_front}
      boxExtents=<10 5 5>
      constraintSetup {angularLimitsMax=<15 0 45>, angularLimitsMin=<-15 -45 -45>, angularTarget=<0 0 0>}
      localJointOffset=<2 0 0>
     }]
    relativeSpaceBone {boneName=pelvis}
   }
   n13 = Anim Dynamics - Bone: dyn_dogtag_chain_04(ComponentPose=@n12, Alpha=1, ExternalForce="0,0,0") {
    angularDampingOverride=0.998
    bGravityOverrideInSimSpace=true
    bOverrideAngularDamping=true
    bUseGravityOverride=true
    simulationSpace=RootRelative
    alphaBoolBlend {blendOption=Linear}
    boundBone {boneName=dyn_dogtag_chain_04}
    componentAppliedLinearAccClamp=<3000 3000 3000>
    componentLinearAccScale=<0.3 0.3 0.3>
    componentLinearVelScale=<0.3 0.3 0.3>
    gravityOverride {z=-980}
    physicsBodyDefinitions [{
      sphereCollisionRadius=0
      boundBone {boneName=dyn_dogtag_chain_04}
      boxExtents=<5 10 10>
      constraintSetup {angularLimitsMax=<1 0 10>, angularLimitsMin=<-1 -45 -10>, angularTarget=<0 0 0>}
      localJointOffset=<2 0 0>
     }]
    relativeSpaceBone {boneName=spine_04}
   }
   n14 = Anim Dynamics - Bone: dyn_dogtag_chain_05(ComponentPose=@n13, Alpha=1, ExternalForce="0,0,0") {
    angularDampingOverride=0.998
    bGravityOverrideInSimSpace=true
    bOverrideAngularDamping=true
    bUseGravityOverride=true
    simulationSpace=RootRelative
    alphaBoolBlend {blendOption=Linear}
    boundBone {boneName=dyn_dogtag_chain_05}
    componentAppliedLinearAccClamp=<3000 3000 3000>
    componentLinearAccScale=<0.3 0.3 0.3>
    componentLinearVelScale=<0.3 0.3 0.3>
    gravityOverride {z=-980}
    physicsBodyDefinitions [{
      sphereCollisionRadius=0
      boundBone {boneName=dyn_dogtag_chain_05}
      boxExtents=<5 10 10>
      constraintSetup {angularLimitsMax=<1 0 15>, angularLimitsMin=<-1 -45 -15>, angularTarget=<0 0 0>}
      localJointOffset=<2 0 0>
     }]
    relativeSpaceBone {boneName=spine_04}
   }
   n15 = Anim Dynamics - Bone: dyn_dogtag_l(ComponentPose=@n14, Alpha=1, ExternalForce="0,0,0") {
    angularDampingOverride=0.998
    bGravityOverrideInSimSpace=true
    bOverrideAngularDamping=true
    bUseGravityOverride=true
    simulationSpace=RootRelative
    alphaBoolBlend {blendOption=Linear}
    boundBone {boneName=dyn_dogtag_l}
    componentAppliedLinearAccClamp=<3000 3000 3000>
    componentLinearAccScale=<0.3 0.3 0.3>
    componentLinearVelScale=<0.3 0.3 0.3>
    gravityOverride {z=-980}
    physicsBodyDefinitions [{
      sphereCollisionRadius=0
      boundBone {boneName=dyn_dogtag_l}
      boxExtents=<10 5 5>
      constraintSetup {angularLimitsMax=<1 0 15>, angularLimitsMin=<-1 -45 -35>, angularTarget=<0 0 0>}
      localJointOffset=<2 0 0>
     }]
    relativeSpaceBone {boneName=spine_04}
   }
   n16 = Anim Dynamics - Bone: dyn_dogtag_r(ComponentPose=@n15, Alpha=1, ExternalForce="0,0,0") {
    angularDampingOverride=0.998
    bGravityOverrideInSimSpace=true
    bOverrideAngularDamping=true
    bUseGravityOverride=true
    simulationSpace=RootRelative
    alphaBoolBlend {blendOption=Linear}
    boundBone {boneName=dyn_dogtag_r}
    componentAppliedLinearAccClamp=<3000 3000 3000>
    componentLinearAccScale=<0.3 0.3 0.3>
    componentLinearVelScale=<0.3 0.3 0.3>
    gravityOverride {z=-980}
    physicsBodyDefinitions [{
      sphereCollisionRadius=0
      boundBone {boneName=dyn_dogtag_r}
      boxExtents=<10 5 5>
      constraintSetup {angularLimitsMax=<1 0 35>, angularLimitsMin=<-1 -45 -15>, angularTarget=<0 0 0>}
      localJointOffset=<2 0 0>
     }]
    relativeSpaceBone {boneName=spine_04}
   }
   n17 = Anim Dynamics - Bone: dyn_holster_strap_fr_buckle_01(ComponentPose=@n16, Alpha=1, ExternalForce="0,0,0") {
    angularDampingOverride=0.98
    angularSpringConstant=75
    bAngularSpring=true
    bGravityOverrideInSimSpace=true
    bOverrideAngularDamping=true
    simulationSpace=RootRelative
    alphaBoolBlend {blendOption=Linear}
    boundBone {boneName=dyn_holster_strap_fr_buckle_01}
    componentAppliedLinearAccClamp=<3000 3000 3000>
    componentLinearAccScale=<1.25 1.25 1.25>
    componentLinearVelScale=<1.25 1.25 1.25>
    gravityOverride {z=-980}
    physicsBodyDefinitions [{
      sphereCollisionRadius=0
      boundBone {boneName=dyn_holster_strap_fr_buckle_01}
      boxExtents=<10 5 5>
      constraintSetup {angularLimitsMax=<1 -1 3>, angularLimitsMin=<-1 -30 -3>, angularTarget=<0 0 0>}
      localJointOffset=<1.5 0 0>
     }]
   }
   n18 = Anim Dynamics - Bone: dyn_holster_strap_fr_03(ComponentPose=@n17, Alpha=1, ExternalForce="0,0,0") {
    angularDampingOverride=0.98
    angularSpringConstant=75
    bAngularSpring=true
    bGravityOverrideInSimSpace=true
    bOverrideAngularDamping=true
    simulationSpace=RootRelative
    alphaBoolBlend {blendOption=Linear}
    boundBone {boneName=dyn_holster_strap_fr_03}
    componentAppliedLinearAccClamp=<3000 3000 3000>
    componentLinearAccScale=<1.25 1.25 1.25>
    componentLinearVelScale=<1.25 1.25 1.25>
    gravityOverride {z=-980}
    physicsBodyDefinitions [{
      sphereCollisionRadius=0
      boundBone {boneName=dyn_holster_strap_fr_03}
      boxExtents=<10 5 5>
      constraintSetup {angularLimitsMax=<1 -1 5>, angularLimitsMin=<-1 -20 -3>, angularTarget=<0 0 0>}
      localJointOffset=<1.5 0 0>
     }]
   }
   n19 = Anim Dynamics - Bone: dyn_holster_strap_fr_04(ComponentPose=@n18, Alpha=1, ExternalForce="0,0,0") {
    angularDampingOverride=0.98
    angularSpringConstant=75
    bAngularSpring=true
    bGravityOverrideInSimSpace=true
    bOverrideAngularDamping=true
    simulationSpace=RootRelative
    alphaBoolBlend {blendOption=Linear}
    boundBone {boneName=dyn_holster_strap_fr_04}
    componentAppliedLinearAccClamp=<3000 3000 3000>
    componentLinearAccScale=<1.25 1.25 1.25>
    componentLinearVelScale=<1.25 1.25 1.25>
    gravityOverride {z=-980}
    physicsBodyDefinitions [{
      sphereCollisionRadius=0
      boundBone {boneName=dyn_holster_strap_fr_04}
      boxExtents=<10 5 5>
      constraintSetup {angularLimitsMax=<1 0 5>, angularLimitsMin=<-1 -5 -3>, angularTarget=<0 0 0>}
      localJointOffset=<1.5 0 0>
     }]
   }
   n20 = Anim Dynamics - Bone: dyn_holster_strap_bk_buckle_01(ComponentPose=@n19, Alpha=1, ExternalForce="0,0,0") {
    angularDampingOverride=0.98
    angularSpringConstant=75
    bAngularSpring=true
    bGravityOverrideInSimSpace=true
    bOverrideAngularDamping=true
    simulationSpace=RootRelative
    alphaBoolBlend {blendOption=Linear}
    boundBone {boneName=dyn_holster_strap_bk_buckle_01}
    componentAppliedLinearAccClamp=<3000 3000 3000>
    componentLinearAccScale=<1.25 1.25 1.25>
    componentLinearVelScale=<1.25 1.25 1.25>
    gravityOverride {z=-980}
    physicsBodyDefinitions [{
      sphereCollisionRadius=0
      boundBone {boneName=dyn_holster_strap_bk_buckle_01}
      boxExtents=<10 5 5>
      constraintSetup {angularLimitsMax=<1 0 3>, angularLimitsMin=<-1 -30 -3>, angularTarget=<0 0 0>}
      localJointOffset=<2 0 0>
     }]
   }
   n21 = Anim Dynamics - Bone: dyn_holster_strap_bk_03(ComponentPose=@n20, Alpha=1, ExternalForce="0,0,0") {
    angularDampingOverride=0.98
    angularSpringConstant=75
    bAngularSpring=true
    bGravityOverrideInSimSpace=true
    bOverrideAngularDamping=true
    simulationSpace=RootRelative
    alphaBoolBlend {blendOption=Linear}
    boundBone {boneName=dyn_holster_strap_bk_03}
    componentAppliedLinearAccClamp=<3000 3000 3000>
    componentLinearAccScale=<1.25 1.25 1.25>
    componentLinearVelScale=<1.25 1.25 1.25>
    gravityOverride {z=-980}
    physicsBodyDefinitions [{
      sphereCollisionRadius=0
      boundBone {boneName=dyn_holster_strap_bk_03}
      boxExtents=<10 5 5>
      constraintSetup {angularLimitsMax=<1 0 3>, angularLimitsMin=<-1 -20 -3>, angularTarget=<0 0 0>}
      localJointOffset=<2 0 0>
     }]
   }
   n22 = Anim Dynamics - Bone: dyn_holster_strap_bk_04(ComponentPose=@n21, Alpha=1, ExternalForce="0,0,0") {
    angularDampingOverride=0.98
    angularSpringConstant=75
    bAngularSpring=true
    bGravityOverrideInSimSpace=true
    bOverrideAngularDamping=true
    simulationSpace=RootRelative
    alphaBoolBlend {blendOption=Linear}
    boundBone {boneName=dyn_holster_strap_bk_04}
    componentAppliedLinearAccClamp=<3000 3000 3000>
    componentLinearAccScale=<1.25 1.25 1.25>
    componentLinearVelScale=<1.25 1.25 1.25>
    gravityOverride {z=-980}
    physicsBodyDefinitions [{
      sphereCollisionRadius=0
      boundBone {boneName=dyn_holster_strap_bk_04}
      boxExtents=<10 5 5>
      constraintSetup {angularLimitsMax=<1 0 3>, angularLimitsMin=<-1 -5 -3>, angularTarget=<0 0 0>}
      localJointOffset=<2 0 0>
     }]
   }
   n23 = Anim Dynamics - Bone: dyn_arm_r_strap_01(ComponentPose=@n22, Alpha=1, ExternalForce="0,0,0") {
    angularDampingOverride=0.98
    angularSpringConstant=300
    bAngularSpring=true
    bGravityOverrideInSimSpace=true
    bOverrideAngularDamping=true
    simulationSpace=RootRelative
    alphaBoolBlend {blendOption=Linear}
    boundBone {boneName=dyn_arm_r_strap_01}
    componentLinearAccScale=<1.25 1.25 1.25>
    componentLinearVelScale=<1.25 1.25 1.25>
    physicsBodyDefinitions [{
      sphereCollisionRadius=0
      boundBone {boneName=dyn_arm_r_strap_01}
      boxExtents=<5 15 15>
      constraintSetup {angularLimitsMax=<2 1 10>, angularLimitsMin=<-2 -25 -10>, angularTarget=<0 0 0>}
      localJointOffset=<1.5 0 0>
     }]
   }
   n24 = Anim Dynamics - Bone: dyn_arm_r_strap_02(ComponentPose=@n23, Alpha=1, ExternalForce="0,0,0") {
    angularDampingOverride=0.98
    angularSpringConstant=300
    bAngularSpring=true
    bGravityOverrideInSimSpace=true
    bOverrideAngularDamping=true
    simulationSpace=RootRelative
    alphaBoolBlend {blendOption=Linear}
    boundBone {boneName=dyn_arm_r_strap_02}
    componentLinearAccScale=<1.25 1.25 1.25>
    componentLinearVelScale=<1.25 1.25 1.25>
    physicsBodyDefinitions [{
      sphereCollisionRadius=0
      boundBone {boneName=dyn_arm_r_strap_02}
      boxExtents=<5 15 15>
      constraintSetup {angularLimitsMax=<2 25 25>, angularLimitsMin=<-2 -25 -10>, angularTarget=<0 0 0>}
      localJointOffset=<1.5 0 0>
     }]
   }
   n25 = Anim Dynamics - Bone: dyn_thigh_r_pouch_strap_01(ComponentPose=@n24, Alpha=1, ExternalForce="0,0,0") {
    angularDampingOverride=0.98
    angularSpringConstant=300
    bAngularSpring=true
    bGravityOverrideInSimSpace=true
    bOverrideAngularDamping=true
    simulationSpace=RootRelative
    alphaBoolBlend {blendOption=Linear}
    boundBone {boneName=dyn_thigh_r_pouch_strap_01}
    componentLinearAccScale=<1.25 1.25 1.25>
    componentLinearVelScale=<1.25 1.25 1.25>
    physicsBodyDefinitions [{
      sphereCollisionRadius=0
      boundBone {boneName=dyn_thigh_r_pouch_strap_01}
      boxExtents=<5 20 20>
      constraintSetup {angularLimitsMax=<2 1 5>, angularLimitsMin=<-2 -25 -5>, angularTarget=<0 0 0>}
      localJointOffset=<1.5 0 0>
     }]
   }
   n26 = Anim Dynamics - Bone: dyn_thigh_r_pouch_strap_02(ComponentPose=@n25, Alpha=1, ExternalForce="0,0,0") {
    angularDampingOverride=0.98
    angularSpringConstant=300
    bAngularSpring=true
    bGravityOverrideInSimSpace=true
    bOverrideAngularDamping=true
    simulationSpace=RootRelative
    alphaBoolBlend {blendOption=Linear}
    boundBone {boneName=dyn_thigh_r_pouch_strap_02}
    componentLinearAccScale=<1.25 1.25 1.25>
    componentLinearVelScale=<1.25 1.25 1.25>
    physicsBodyDefinitions [{
      sphereCollisionRadius=0
      boundBone {boneName=dyn_thigh_r_pouch_strap_02}
      boxExtents=<5 15 15>
      constraintSetup {angularLimitsMax=<2 1 10>, angularLimitsMin=<-2 -25 -10>, angularTarget=<0 0 0>}
      localJointOffset=<1.5 0 0>
     }]
   }
   n27 = Anim Dynamics - Bone: dyn_thigh_r_pouch_strap_03(ComponentPose=@n26, Alpha=1, ExternalForce="0,0,0") {
    angularDampingOverride=0.98
    angularSpringConstant=300
    bAngularSpring=true
    bGravityOverrideInSimSpace=true
    bOverrideAngularDamping=true
    simulationSpace=RootRelative
    alphaBoolBlend {blendOption=Linear}
    boundBone {boneName=dyn_thigh_r_pouch_strap_03}
    componentLinearAccScale=<1.25 1.25 1.25>
    componentLinearVelScale=<1.25 1.25 1.25>
    physicsBodyDefinitions [{
      sphereCollisionRadius=0
      boundBone {boneName=dyn_thigh_r_pouch_strap_03}
      boxExtents=<5 15 15>
      constraintSetup {angularLimitsMax=<2 25 25>, angularLimitsMin=<-2 -25 -10>, angularTarget=<0 0 0>}
      localJointOffset=<1.5 0 0>
     }]
   }
   n28 = Anim Dynamics - Bone: dyn_knife(ComponentPose=@n27, Alpha=1, ExternalForce="0,0,0") {
    angularDampingOverride=0.98
    angularSpringConstant=300
    bAngularSpring=true
    bGravityOverrideInSimSpace=true
    bOverrideAngularDamping=true
    simulationSpace=RootRelative
    alphaBoolBlend {blendOption=Linear}
    boundBone {boneName=dyn_knife}
    componentLinearAccScale=<1.25 1.25 1.25>
    componentLinearVelScale=<1.25 1.25 1.25>
    physicsBodyDefinitions [{
      sphereCollisionRadius=0
      boundBone {boneName=dyn_knife}
      boxExtents=<5 10 20>
      constraintSetup {angularLimitsMax=<2 1 10>, angularLimitsMin=<-2 -1 -10>, angularTarget=<0 0 0>}
      localJointOffset=<0.5 0 -3>
     }]
   }
   n29 = Anim Dynamics - Bone: dyn_shell_01(ComponentPose=@n28, Alpha=1, ExternalForce="0,0,0") {
    angularDampingOverride=0.98
    angularSpringConstant=100
    bAngularSpring=true
    bGravityOverrideInSimSpace=true
    bOverrideAngularDamping=true
    simulationSpace=RootRelative
    alphaBoolBlend {blendOption=Linear}
    boundBone {boneName=dyn_shell_01}
    componentLinearAccScale=<1.25 1.25 1.25>
    componentLinearVelScale=<1.25 1.25 1.25>
    physicsBodyDefinitions [{
      sphereCollisionRadius=0
      boundBone {boneName=dyn_shell_01}
      boxExtents=<3 10 10>
      constraintSetup {angularLimitsMax=<2 5 1>, angularLimitsMin=<-2 -2 -1>, angularTarget=<0 0 0>}
      localJointOffset=<0.5 0 1>
     }]
   }
   n30 = Anim Dynamics - Bone: dyn_shell_02(ComponentPose=@n29, Alpha=1, ExternalForce="0,0,0") {
    angularDampingOverride=0.98
    angularSpringConstant=150
    bAngularSpring=true
    bGravityOverrideInSimSpace=true
    bOverrideAngularDamping=true
    simulationSpace=RootRelative
    alphaBoolBlend {blendOption=Linear}
    boundBone {boneName=dyn_shell_02}
    componentLinearAccScale=<1.25 1.25 1.25>
    componentLinearVelScale=<1.25 1.25 1.25>
    physicsBodyDefinitions [{
      sphereCollisionRadius=0
      boundBone {boneName=dyn_shell_02}
      boxExtents=<3 10 10>
      constraintSetup {angularLimitsMax=<3 3 1>, angularLimitsMin=<-1 -3 -1>, angularTarget=<0 0 0>}
      localJointOffset=<1 0 -3>
     }]
   }
   n31 = Anim Dynamics - Bone: dyn_shell_03(ComponentPose=@n30, Alpha=1, ExternalForce="0,0,0") {
    angularDampingOverride=0.98
    angularSpringConstant=200
    bAngularSpring=true
    bGravityOverrideInSimSpace=true
    bOverrideAngularDamping=true
    simulationSpace=RootRelative
    alphaBoolBlend {blendOption=Linear}
    boundBone {boneName=dyn_shell_03}
    componentLinearAccScale=<1.25 1.25 1.25>
    componentLinearVelScale=<1.25 1.25 1.25>
    physicsBodyDefinitions [{
      sphereCollisionRadius=0
      boundBone {boneName=dyn_shell_03}
      boxExtents=<3 10 10>
      constraintSetup {angularLimitsMax=<1 2 3>, angularLimitsMin=<-3 -5 -1>, angularTarget=<0 0 0>}
      localJointOffset=<-0.5 0 3>
     }]
   }
   n32 = "Transform (Modify) Bone - Bone: dyn_glute_l"(
    ComponentPose=Local To Component(LocalPose=Copy Pose From Mesh() {bUseAttachedParent=true})
    bAlphaBoolEnabled=Get bIsCrouching()
    Translation="0.000,-2,-4"
    Rotation="0,0,0"
    Scale="1,1,1"
   ) {
    alphaInputType=Bool
    translationMode=BMM_Additive
    alphaBoolBlend {blendOption=Linear}
    boneToModify {boneName=dyn_glute_l}
    translation {y=-2, z=-4}
   }
   n33 = "Constraint - dyn_holster_strap_fr_MSTR (PPPPP)"(
    ComponentPose=@n32
    Alpha=1
    ConstraintWeights_0=0.207
    ConstraintWeights_1=0.3785
    ConstraintWeights_2=0.0522
    ConstraintWeights_3=0.1464
    ConstraintWeights_4=0.2159
   ) {
    alphaBoolBlend {blendOption=Linear}
    boneToModify {boneName=dyn_holster_strap_fr_MSTR}
    constraintSetup [
     {targetBone {boneName=spine_01}}
     {targetBone {boneName=spine_02}}
     {targetBone {boneName=spine_03}}
     {targetBone {boneName=spine_04}}
     {targetBone {boneName=pelvis}}
    ]
    constraintWeights [1,1,1,1,1]
   }
   n34 = "Constraint - dyn_holster_strap_bk_MSTR (PPPPP)"(
    ComponentPose=@n33
    Alpha=1
    ConstraintWeights_0=0.2524
    ConstraintWeights_1=0.3941
    ConstraintWeights_2=0.0639
    ConstraintWeights_3=0.1081
    ConstraintWeights_4=0.1815
   ) {
    alphaBoolBlend {blendOption=Linear}
    boneToModify {boneName=dyn_holster_strap_bk_MSTR}
    constraintSetup [
     {targetBone {boneName=spine_01}}
     {targetBone {boneName=spine_02}}
     {targetBone {boneName=spine_03}}
     {targetBone {boneName=spine_04}}
     {targetBone {boneName=pelvis}}
    ]
    constraintWeights [1,1,1,1,1]
   }
   n35 = "Constraint - dyn_holster_strap_fr_buckle_02 (R)"(ComponentPose=@n34, Alpha=1, ConstraintWeights_0=1) {
    alphaBoolBlend {blendOption=Linear}
    boneToModify {boneName=dyn_holster_strap_fr_buckle_02}
    constraintSetup [{transformType=Rotation, targetBone {boneName=thigh_r}}]
    constraintWeights [1]
   }
   n36 = "Constraint - dyn_jacket_fr_r_MSTR (PPPPPPPP)"(
    ComponentPose=@n35
    Alpha=1
    ConstraintWeights_0=0.0005
    ConstraintWeights_1=0.0103
    ConstraintWeights_2=0.0542
    ConstraintWeights_3=0.0584
    ConstraintWeights_4=0.0025
    ConstraintWeights_5=0.0111
    ConstraintWeights_6=0.8612
    ConstraintWeights_7=0.0018
   ) {
    alphaBoolBlend {blendOption=Linear}
    boneToModify {boneName=dyn_jacket_fr_r_MSTR}
    constraintSetup [
     {targetBone {boneName=spine_01}}
     {targetBone {boneName=dyn_jacket_r_01}}
     {targetBone {boneName=dyn_jacket_fr_r_01}}
     {targetBone {boneName=dyn_jacket_fr_r_02}}
     {targetBone {boneName=spine_02}}
     {targetBone {boneName=spine_03}}
     {targetBone {boneName=spine_04}}
     {targetBone {boneName=spine_05}}
    ]
    constraintWeights [1,1,1,1,1,1,1,1]
   }
   n37 = "Constraint - dyn_jacket_fr_l_MSTR (PPPPPPPP)"(
    ComponentPose=@n36
    Alpha=1
    ConstraintWeights_0=0.0005
    ConstraintWeights_1=0
    ConstraintWeights_2=0.1775
    ConstraintWeights_3=0
    ConstraintWeights_4=0.0001
    ConstraintWeights_5=0.018
    ConstraintWeights_6=0.8025
    ConstraintWeights_7=0.0014
   ) {
    alphaBoolBlend {blendOption=Linear}
    boneToModify {boneName=dyn_jacket_fr_l_MSTR}
    constraintSetup [
     {targetBone {boneName=spine_03}}
     {targetBone {boneName=clavicle_l}}
     {targetBone {boneName=dyn_jacket_fr_l_01}}
     {targetBone {boneName=spine_01}}
     {targetBone {boneName=spine_02}}
     {targetBone {boneName=dyn_jacket_fr_l_02}}
     {targetBone {boneName=spine_04}}
     {targetBone {boneName=spine_05}}
    ]
    constraintWeights [1,1,1,1,1,1,1,1]
   }
   n38 = Anim Dynamics - Bone: dyn_dogtag_chain_04(ComponentPose=@n37, Alpha=1, ExternalForce="0,0,0") {
    angularDampingOverride=0.998
    bGravityOverrideInSimSpace=true
    bOverrideAngularDamping=true
    bUseGravityOverride=true
    simulationSpace=RootRelative
    alphaBoolBlend {blendOption=Linear}
    boundBone {boneName=dyn_dogtag_chain_04}
    componentAppliedLinearAccClamp=<3000 3000 3000>
    componentLinearAccScale=<0.3 0.3 0.3>
    componentLinearVelScale=<0.3 0.3 0.3>
    gravityOverride {z=-980}
    physicsBodyDefinitions [{
      sphereCollisionRadius=0
      boundBone {boneName=dyn_dogtag_chain_04}
      boxExtents=<5 10 10>
      constraintSetup {angularLimitsMax=<1 0 10>, angularLimitsMin=<-1 -45 -10>, angularTarget=<0 0 0>}
      localJointOffset=<2 0 0>
     }]
    relativeSpaceBone {boneName=spine_04}
   }
   n39 = Anim Dynamics - Bone: dyn_dogtag_chain_05(ComponentPose=@n38, Alpha=1, ExternalForce="0,0,0") {
    angularDampingOverride=0.998
    bGravityOverrideInSimSpace=true
    bOverrideAngularDamping=true
    bUseGravityOverride=true
    simulationSpace=RootRelative
    alphaBoolBlend {blendOption=Linear}
    boundBone {boneName=dyn_dogtag_chain_05}
    componentAppliedLinearAccClamp=<3000 3000 3000>
    componentLinearAccScale=<0.3 0.3 0.3>
    componentLinearVelScale=<0.3 0.3 0.3>
    gravityOverride {z=-980}
    physicsBodyDefinitions [{
      sphereCollisionRadius=0
      boundBone {boneName=dyn_dogtag_chain_05}
      boxExtents=<5 10 10>
      constraintSetup {angularLimitsMax=<1 0 15>, angularLimitsMin=<-1 -45 -15>, angularTarget=<0 0 0>}
      localJointOffset=<2 0 0>
     }]
    relativeSpaceBone {boneName=spine_04}
   }
   n40 = Anim Dynamics - Bone: dyn_dogtag_l(ComponentPose=@n39, Alpha=1, ExternalForce="0,0,0") {
    angularDampingOverride=0.998
    bGravityOverrideInSimSpace=true
    bOverrideAngularDamping=true
    bUseGravityOverride=true
    simulationSpace=RootRelative
    alphaBoolBlend {blendOption=Linear}
    boundBone {boneName=dyn_dogtag_l}
    componentAppliedLinearAccClamp=<3000 3000 3000>
    componentLinearAccScale=<0.3 0.3 0.3>
    componentLinearVelScale=<0.3 0.3 0.3>
    gravityOverride {z=-980}
    physicsBodyDefinitions [{
      sphereCollisionRadius=0
      boundBone {boneName=dyn_dogtag_l}
      boxExtents=<10 5 5>
      constraintSetup {angularLimitsMax=<1 0 15>, angularLimitsMin=<-1 -45 -35>, angularTarget=<0 0 0>}
      localJointOffset=<2 0 0>
     }]
    relativeSpaceBone {boneName=spine_04}
   }
   n41 = Anim Dynamics - Bone: dyn_dogtag_r(ComponentPose=@n40, Alpha=1, ExternalForce="0,0,0") {
    angularDampingOverride=0.998
    bGravityOverrideInSimSpace=true
    bOverrideAngularDamping=true
    bUseGravityOverride=true
    simulationSpace=RootRelative
    alphaBoolBlend {blendOption=Linear}
    boundBone {boneName=dyn_dogtag_r}
    componentAppliedLinearAccClamp=<3000 3000 3000>
    componentLinearAccScale=<0.3 0.3 0.3>
    componentLinearVelScale=<0.3 0.3 0.3>
    gravityOverride {z=-980}
    physicsBodyDefinitions [{
      sphereCollisionRadius=0
      boundBone {boneName=dyn_dogtag_r}
      boxExtents=<10 5 5>
      constraintSetup {angularLimitsMax=<1 0 35>, angularLimitsMin=<-1 -45 -15>, angularTarget=<0 0 0>}
      localJointOffset=<2 0 0>
     }]
    relativeSpaceBone {boneName=spine_04}
   }
   n42 = RigidBody(ComponentPose=@n41) {
    bClampLinearTranslationLimitToRefPose=true
    bOverrideWorldGravity=true
    overridePhysicsAsset=&/F_MED_Street_Fashion_Diamond_Jacket_Physics
    simulationSpace=BaseBoneSpace
    alphaBoolBlend {blendOption=Linear}
    baseBoneRef {boneName=pelvis}
    componentLinearAccScale=<0.5 0.5 0.7>
    componentLinearVelScale=<0.3 0.3 0.5>
    overrideWorldGravity {z=-980}
   }
   n43 = RigidBody(ComponentPose=@n42) {
    bClampLinearTranslationLimitToRefPose=true
    bOverrideWorldGravity=true
    overridePhysicsAsset=&/F_MED_Street_Fashion_Diamond_pouch_physics
    simulationSpace=BaseBoneSpace
    alphaBoolBlend {blendOption=Linear}
    baseBoneRef {boneName=root}
    componentLinearAccScale=<0.3 0.3 0.5>
    componentLinearVelScale=<0.3 0.3 0.3>
    overrideWorldGravity {z=-980}
   }
   n44 = RigidBody(ComponentPose=@n43) {
    bClampLinearTranslationLimitToRefPose=true
    bOverrideWorldGravity=true
    overridePhysicsAsset=&/F_MED_Street_Fashion_Diamond_holster_Physics
    simulationSpace=BaseBoneSpace
    alphaBoolBlend {blendOption=Linear}
    baseBoneRef {boneName=pelvis}
    componentLinearAccScale=<0.3 0.3 0.4>
    componentLinearVelScale=<0.3 0.3 0.4>
    overrideWorldGravity {z=-980}
   }
   n45 = RigidBody(ComponentPose=@n44) {
    bClampLinearTranslationLimitToRefPose=true
    bOverrideWorldGravity=true
    overridePhysicsAsset=&/F_MED_Street_Fashion_Diamond_hood_Physics
    simulationSpace=BaseBoneSpace
    alphaBoolBlend {blendOption=Linear}
    baseBoneRef {boneName=root}
    componentAppliedLinearAccClamp=<5000 5000 5000>
    componentLinearAccScale=<0.3 0.3 0.4>
    componentLinearVelScale=<0.2 0.2 0.2>
    overrideWorldGravity {z=-980}
   }
   Output Pose(Result=Blend Poses by bool(BlendPose_0=Component To Local(ComponentPose=@n45), BlendPose_1=Component To Local(ComponentPose=@n31), bActiveValue=Get bIsInFrontEnd()) {blendType=Linear}) {layerGroup=None}
  ]}
 EventGraph {nodes [Event BlueprintUpdateAnimation()]}
]
variables []
