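# Physics asset for the hat part of the female medium Street Fashion Diamond character, covering dangling hair and a glasses strap.
# A head capsule anchors small boxes on left and right two-bone side hair chains, a back hair master bone and the strap.
# Side hair and strap constraints use tight twist and narrow swing limits.
# The back hair master joint leaves swing and twist free.
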PhysicsAsset /BRCosmetics/Characters/Player/Female/Medium/Bodies/F_MED_Street_Fashion_Diamond/Meshes/Parts/F_MED_Street_Fashion_Diamond_hat_Physics
bodies {
 dyn_bk_hair_MSTR [{shapeType=Box, extentX=1.01, extentY=1.01, extentZ=1.01}]
 dyn_glasses_strap [{shapeType=Box, extentX=2.0686, extentY=2.9292, extentZ=1.01, center=<0.9476 0 0.0001>, rotation=<rot 0 -90 0>}]
 dyn_l_side_hair_01 [{shapeType=Box, extentX=3.0744, extentY=2.8073, extentZ=1.3685, center=<1.3161 0.0219 0.274>}]
 dyn_l_side_hair_02 [{shapeType=Box, extentX=5.2335, extentY=2.7992, extentZ=1.6159, center=<1.4796 0.2121 0.2225>}]
 dyn_r_side_hair_01 [{shapeType=Box, extentX=3.2355, extentY=3.8713, extentZ=1.2247, center=<1.4307 0.4927 0.2463>}]
 dyn_r_side_hair_02 [{shapeType=Box, extentX=4.6856, extentY=2.0687, extentZ=1.2112, center=<1.7805 0.2239 0.1492>}]
 head [{shapeType=Capsule, length=8.9509, radius=6.0184, center=<-1.3461 0.2216 -0.0803>, rotation=<rot 0.3176 -68.8318 -89.8843>}]
}
constraints [
 {bone1Name=dyn_r_side_hair_01, bone2Name=head, swing1LimitDegrees=30, swing2LimitDegrees=20, twistLimitDegrees=1}
 {bone1Name=dyn_l_side_hair_01, bone2Name=head, swing1LimitDegrees=30, swing2LimitDegrees=20, twistLimitDegrees=1}
 {bone1Name=dyn_r_side_hair_02, bone2Name=dyn_r_side_hair_01, swing1LimitDegrees=30, swing2LimitDegrees=20, twistLimitDegrees=1}
 {bone1Name=dyn_l_side_hair_02, bone2Name=dyn_l_side_hair_01, swing1LimitDegrees=30, swing2LimitDegrees=20, twistLimitDegrees=1}
 {
  bone1Name=dyn_bk_hair_MSTR
  bone2Name=head
  swing1LimitDegrees=45
  swing1Motion=Free
  swing2LimitDegrees=45
  swing2Motion=Free
  twistLimitDegrees=45
  twistMotion=Free
 }
 {bone1Name=dyn_glasses_strap, bone2Name=head, swing1LimitDegrees=3, swing2LimitDegrees=30, twistLimitDegrees=1}
]
