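# Physics asset for the F_MED_Street_Fashion_Diamond female character part, setting up hair dynamics against upper-body collision shapes.
# Capsules cover spine, clavicles, upper arms and neck, with a thin, wide box on the head.
# Three five-bone back chains and two three-bone side chains hang from box-shaped master bodies, linked to the head through the back master.
# Body and master joints swing freely, while strand joints have angular limits, mostly loosening toward the tips.
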PhysicsAsset /BRCosmetics/Characters/Player/Female/Medium/Bodies/F_MED_Street_Fashion_Diamond/Meshes/Parts/F_MED_Street_Fashion_Diamond_new_Physics
bodies {
 clavicle_l [{shapeType=Capsule, length=3.6894, radius=7.0419, center=<6.4482 -0.3686 -2.4355>, rotation=<rot 85.606 0 0>}]
 clavicle_r [{shapeType=Capsule, length=3.4794, radius=7.2519, center=<-6.4482 0.3686 2.4355>, rotation=<rot 85.606 0 0>}]
 dyn_bk_hair_01 [{shapeType=Capsule, length=1.4461, radius=1.9998, center=<1.5 0 1>, rotation=<rot 0 0 90>}]
 dyn_bk_hair_02 [{shapeType=Capsule, length=2.1928, radius=1.6841, center=<1.5 0 1>, rotation=<rot 0 0 90>}]
 dyn_bk_hair_03 [{shapeType=Capsule, length=1.1895, radius=2.3156, center=<1.5 0 1>, rotation=<rot 0 0 90>}]
 dyn_bk_hair_04 [{shapeType=Capsule, length=1.869, radius=2.1757, center=<1.5 0 1>, rotation=<rot 0 0 90>}]
 dyn_bk_hair_05 [{shapeType=Capsule, length=2.4936, radius=2.1181, center=<1.7 0 0.5>, rotation=<rot 0 0 90>}]
 dyn_bk_hair_MSTR [{shapeType=Box, extentX=1.01, extentY=1.01, extentZ=1.01, rotation=<rot 0 -90 0>}]
 dyn_bk_l_hair_01 [{shapeType=Capsule, length=1.6332, radius=1.9347, center=<1.5 0 0.7>, rotation=<rot 0 0 90>}]
 dyn_bk_l_hair_02 [{shapeType=Capsule, length=2.4059, radius=1.7804, center=<1.5 0 0.7>, rotation=<rot 0 0 90>}]
 dyn_bk_l_hair_03 [{shapeType=Capsule, length=1.0701, radius=2.1616, center=<1.7 0 0.7>, rotation=<rot 0 0 90>}]
 dyn_bk_l_hair_04 [{shapeType=Capsule, length=2.7653, radius=1.6441, center=<1.5 0 0.5>, rotation=<rot 0 0 119.9987>}]
 dyn_bk_l_hair_05 [{shapeType=Capsule, length=1.7582, radius=2.0145, center=<1.5 0 0.5>, rotation=<rot 0 0 105.0006>}]
 dyn_bk_r_hair_01 [{shapeType=Capsule, length=2.2731, radius=1.7965, center=<1.5 0 0.7>, rotation=<rot 0 0 90>}]
 dyn_bk_r_hair_02 [{shapeType=Capsule, length=2.8141, radius=1.8012, center=<2 0 0.7>, rotation=<rot 0 0 90>}]
 dyn_bk_r_hair_03 [{shapeType=Capsule, length=2.6024, radius=2.224, center=<2 0 0.7>, rotation=<rot 0 0 79.9999>}]
 dyn_bk_r_hair_04 [{shapeType=Capsule, length=3.05, radius=2.1066, center=<2.5 0 0.5>, rotation=<rot 0 0 84.9998>}]
 dyn_bk_r_hair_05 [{shapeType=Capsule, length=2.6762, radius=2.0098, center=<2 0 0.5>, rotation=<rot 0 0 85>}]
 dyn_l_hair_01 [{shapeType=Capsule, length=2.155, radius=1.5954, center=<2.5 0 0>, rotation=<rot 90 0 0>}]
 dyn_l_hair_02 [{shapeType=Capsule, length=1.9544, radius=1.4564, center=<2.2 0 -0.5>, rotation=<rot 90 0 0>}]
 dyn_l_hair_03 [{shapeType=Capsule, length=0.6577, radius=1.2847, center=<1.596 0 -0.5>, rotation=<rot 90 0 0>}]
 dyn_l_hair_MSTR [{shapeType=Box, extentX=1.01, extentY=1.01, extentZ=1.01, rotation=<rot 0 -90 0>}]
 dyn_r_hair_01 [{shapeType=Capsule, length=2.2181, radius=1.591, center=<2.5 0 0>, rotation=<rot 90 0 0>}]
 dyn_r_hair_02 [{shapeType=Capsule, length=1.4265, radius=1.6947, center=<1.8 0 -0.2>, rotation=<rot 90 0 0>}]
 dyn_r_hair_03 [{shapeType=Capsule, length=2.5696, radius=1.5881, center=<2.7284 0 -0.2>, rotation=<rot 79.9922 -179.9999 174.9998>}]
 dyn_r_hair_MSTR [{shapeType=Box, extentX=1.01, extentY=1.01, extentZ=1.01, rotation=<rot 0 -90 0>}]
 head [{shapeType=Box, extentX=2.688, extentY=40.6292, extentZ=41.5956, center=<-1.1394 1.4284 -0.0053>, rotation=<rot 0.9616 120.6881 -88.8529>}]
 neck_01 [{shapeType=Capsule, length=5.5017, radius=6.7506, center=<-2.2411 0.1685 -2.6772>, rotation=<rot 0 -90 0>}]
 neck_02 [{shapeType=Capsule, length=6.7179, radius=5.46, center=<3.675 0.1935 0>, rotation=<rot 90 0 -12.1655>}]
 spine_05 [
  {shapeType=Capsule, length=7.4351, radius=11.4901, center=<0.5981 1.033 0>}
  {shapeType=Capsule, length=9.8497, radius=7.2239, center=<8.7466 -0.306 0.0209>}
 ]
 upperarm_l [{shapeType=Capsule, length=18.4886, radius=4.5986, center=<10.8056 -0.4936 0.3456>, rotation=<rot 90 0 0>}]
 upperarm_r [{shapeType=Capsule, length=18.4886, radius=4.5986, center=<-10.8056 0.4936 -0.3456>, rotation=<rot 90 0 0>}]
}
constraints [
 {
  bone1Name=dyn_r_hair_MSTR
  bone2Name=dyn_bk_hair_MSTR
  swing1LimitDegrees=45
  swing1Motion=Free
  swing2LimitDegrees=45
  swing2Motion=Free
  twistLimitDegrees=45
  twistMotion=Free
 }
 {
  bone1Name=dyn_l_hair_MSTR
  bone2Name=dyn_bk_hair_MSTR
  swing1LimitDegrees=45
  swing1Motion=Free
  swing2LimitDegrees=45
  swing2Motion=Free
  twistLimitDegrees=45
  twistMotion=Free
 }
 {bone1Name=dyn_bk_hair_01, bone2Name=dyn_bk_hair_MSTR, swing1LimitDegrees=20, swing2LimitDegrees=45, twistLimitDegrees=5}
 {bone1Name=dyn_bk_hair_02, bone2Name=dyn_bk_hair_01, swing1LimitDegrees=20, swing2LimitDegrees=45, twistLimitDegrees=5}
 {bone1Name=dyn_bk_hair_03, bone2Name=dyn_bk_hair_02, swing1LimitDegrees=20, swing2LimitDegrees=45, twistLimitDegrees=5}
 {bone1Name=dyn_bk_hair_04, bone2Name=dyn_bk_hair_03, swing1LimitDegrees=20, swing2LimitDegrees=45, twistLimitDegrees=5}
 {bone1Name=dyn_bk_hair_05, bone2Name=dyn_bk_hair_04, swing1LimitDegrees=20, swing2LimitDegrees=45, twistLimitDegrees=5}
 {
  bone1Name=dyn_bk_hair_MSTR
  bone2Name=head
  swing1LimitDegrees=45
  swing1Motion=Free
  swing2LimitDegrees=45
  swing2Motion=Free
  twistLimitDegrees=45
  twistMotion=Free
 }
 {
  bone1Name=clavicle_l
  bone2Name=spine_05
  swing1LimitDegrees=45
  swing1Motion=Free
  swing2LimitDegrees=45
  swing2Motion=Free
  twistLimitDegrees=45
  twistMotion=Free
 }
 {
  bone1Name=clavicle_r
  bone2Name=spine_05
  swing1LimitDegrees=45
  swing1Motion=Free
  swing2LimitDegrees=45
  swing2Motion=Free
  twistLimitDegrees=45
  twistMotion=Free
 }
 {bone1Name=dyn_bk_l_hair_01, bone2Name=dyn_bk_hair_MSTR, swing1LimitDegrees=10, swing2LimitDegrees=45, twistLimitDegrees=5}
 {bone1Name=dyn_bk_l_hair_02, bone2Name=dyn_bk_l_hair_01, swing1LimitDegrees=15, swing2LimitDegrees=45, twistLimitDegrees=5}
 {bone1Name=dyn_bk_l_hair_03, bone2Name=dyn_bk_l_hair_02, swing1LimitDegrees=20, swing2LimitDegrees=45, twistLimitDegrees=5}
 {bone1Name=dyn_bk_l_hair_04, bone2Name=dyn_bk_l_hair_03, swing1LimitDegrees=25, swing2LimitDegrees=45, twistLimitDegrees=5}
 {bone1Name=dyn_bk_l_hair_05, bone2Name=dyn_bk_l_hair_04, swing1LimitDegrees=30, swing2LimitDegrees=45, twistLimitDegrees=5}
 {bone1Name=dyn_bk_r_hair_01, bone2Name=dyn_bk_hair_MSTR, swing1LimitDegrees=15, swing2LimitDegrees=45, twistLimitDegrees=5}
 {bone1Name=dyn_bk_r_hair_02, bone2Name=dyn_bk_r_hair_01, swing1LimitDegrees=20, swing2LimitDegrees=45, twistLimitDegrees=5}
 {bone1Name=dyn_bk_r_hair_03, bone2Name=dyn_bk_r_hair_02, swing1LimitDegrees=25, swing2LimitDegrees=45, twistLimitDegrees=5}
 {bone1Name=dyn_bk_r_hair_04, bone2Name=dyn_bk_r_hair_03, swing1LimitDegrees=30, swing2LimitDegrees=45, twistLimitDegrees=5}
 {bone1Name=dyn_bk_r_hair_05, bone2Name=dyn_bk_r_hair_04, swing1LimitDegrees=30, swing2LimitDegrees=45, twistLimitDegrees=5}
 {bone1Name=dyn_r_hair_01, bone2Name=dyn_r_hair_MSTR, swing1LimitDegrees=10, swing2LimitDegrees=55, twistLimitDegrees=30}
 {bone1Name=dyn_r_hair_02, bone2Name=dyn_r_hair_01, swing1LimitDegrees=15, swing2LimitDegrees=55, twistLimitDegrees=30}
 {bone1Name=dyn_r_hair_03, bone2Name=dyn_r_hair_02, swing1LimitDegrees=20, swing2LimitDegrees=45, twistLimitDegrees=30}
 {bone1Name=dyn_l_hair_01, bone2Name=dyn_l_hair_MSTR, swing1LimitDegrees=5, swing2LimitDegrees=45, twistLimitDegrees=15}
 {bone1Name=dyn_l_hair_02, bone2Name=dyn_l_hair_01, swing1LimitDegrees=10, swing2LimitDegrees=45, twistLimitDegrees=15}
 {bone1Name=dyn_l_hair_03, bone2Name=dyn_l_hair_02, swing1LimitDegrees=10, swing2LimitDegrees=45, twistLimitDegrees=15}
 {
  bone1Name=neck_01
  bone2Name=spine_05
  swing1LimitDegrees=45
  swing1Motion=Free
  swing2LimitDegrees=45
  swing2Motion=Free
  twistLimitDegrees=45
  twistMotion=Free
 }
 {
  bone1Name=upperarm_l
  bone2Name=clavicle_l
  swing1LimitDegrees=45
  swing1Motion=Free
  swing2LimitDegrees=45
  swing2Motion=Free
  twistLimitDegrees=45
  twistMotion=Free
 }
 {
  bone1Name=upperarm_r
  bone2Name=clavicle_r
  swing1LimitDegrees=45
  swing1Motion=Free
  swing2LimitDegrees=45
  swing2Motion=Free
  twistLimitDegrees=45
  twistMotion=Free
 }
 {
  bone1Name=head
  bone2Name=neck_02
  swing1LimitDegrees=45
  swing1Motion=Free
  swing2LimitDegrees=45
  swing2Motion=Free
  twistLimitDegrees=45
  twistMotion=Free
 }
 {
  bone1Name=neck_02
  bone2Name=neck_01
  swing1LimitDegrees=45
  swing1Motion=Free
  swing2LimitDegrees=45
  swing2Motion=Free
  twistLimitDegrees=45
  twistMotion=Free
 }
]
