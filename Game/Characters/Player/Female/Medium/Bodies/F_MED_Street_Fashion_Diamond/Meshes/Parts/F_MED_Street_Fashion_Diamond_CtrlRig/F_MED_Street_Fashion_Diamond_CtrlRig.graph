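# Control Rig for the F_MED_Street_Fashion_Diamond player character part that offsets hair bones as the head bends.
# It reads the head's rotation relative to neck_01 as Euler Z and maps a negative angle range through clamped remaps.
# The results drive additive local rotation and translation on dyn_bk_hair_MSTR, then on its child dyn_r_hair_MSTR.
# A head transform read and an inverse 1-to-0 remap are left unconnected.
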ControlRigBlueprint /Game/Characters/Player/Female/Medium/Bodies/F_MED_Street_Fashion_Diamond/Meshes/Parts/F_MED_Street_Fashion_Diamond_CtrlRig/F_MED_Street_Fashion_Diamond_CtrlRig
graphs [
 RigVMFunctionLibrary {links [], nodes []}
 RigVMModel {nodes [
   RigUnit_BeginExecution = &RigUnit_BeginExecution()
   n1 = &RigVMFunction_MathQuaternionToEuler(RotationOrder=ZYX, Value=&RigUnit_GetRelativeBoneTransform(Bone=head, Space=neck_01).Transform.Rotation)
   /Script/ControlRig.RigUnit_ModifyBoneTransforms(
    BoneToModify=((Bone="dyn_bk_hair_MSTR",Transform=(Rotation=(X=0,Y=0,Z=0,W=1),Translation=(X=0,Y=0,Z=0),Scale3D=(X=1,Y=1,Z=1))))
    Mode=AdditiveLocal
    Weight=1
    WeightMaximum=1
    BoneToModify.0.Transform.Translation.Y=&RigVMFunction_MathFloatRemap(SourceMaximum=-50, SourceMinimum=-19, TargetMaximum=-1, TargetMinimum=0, bClamp=true, Value=@n1.Result.Z).Result
    BoneToModify.0.Transform.Rotation=&RigVMFunction_MathQuaternionFromEuler(
     Euler=(X=0,Y=0,Z=0)
     Euler.Z=&RigVMFunction_MathFloatRemap(SourceMaximum=-50, SourceMinimum=-19, TargetMaximum=20, TargetMinimum=0, bClamp=true, Value=@n1.Result.Z).Result
    ).Result
    ExecutePin=@RigUnit_BeginExecution.ExecutePin
   ) {is=struct RigVMUnitNode}
   > &RigUnit_ModifyBoneTransforms(
    BoneToModify=((Bone="dyn_r_hair_MSTR",Transform=(Rotation=(X=0,Y=0,Z=0,W=1),Translation=(X=0,Y=0,Z=0),Scale3D=(X=1,Y=1,Z=1))))
    Mode=AdditiveLocal
    Weight=1
    WeightMaximum=1
    BoneToModify.0.Transform.Translation.Z=&RigVMFunction_MathFloatRemap(SourceMaximum=-50, SourceMinimum=-19, TargetMaximum=-2, TargetMinimum=0, bClamp=true, Value=@n1.Result.Z).Result
    BoneToModify.0.Transform.Rotation=&RigVMFunction_MathQuaternionFromEuler(
     Euler=(X=0,Y=0,Z=0)
     Euler.Z=&RigVMFunction_MathFloatRemap(SourceMaximum=-50, SourceMinimum=-19, TargetMaximum=15, TargetMinimum=0, bClamp=true, Value=@n1.Result.Z).Result
    ).Result
   )
   &RigUnit_GetBoneTransform(Bone=head, Space=LocalSpace)
   &RigVMFunction_MathFloatRemap(SourceMaximum=-50, SourceMinimum=-19, TargetMaximum=0, TargetMinimum=1, bClamp=true, Value=@n1.Result.Z)
  ]}
]
hierarchy [
 root {loc=<0 0 0>, rot=<0 0 0 1>}
 pelvis {parent=root, loc=<0 -0.5301 98.0571>, rot=<0.0127 0.707 -0.0127 -0.707>}
 spine_01 {parent=pelvis, loc=<6.1208 -0.9328 0>, rot=<0 0 0.034 -0.9994>}
 spine_02 {parent=spine_01, loc=<6.5996 -0.2095 0>, rot=<0 0 0.0238 -0.9997>}
 spine_03 {parent=spine_02, loc=<6.4914 -0.1237 0>, rot=<0 0 0.0149 -0.9999>}
 spine_04 {parent=spine_03, loc=<6.9598 0.0794 0>, rot=<0 0 -0.0132 -0.9999>}
 spine_05 {parent=spine_04, loc=<6.9987 0 0>, rot=<0 0 -0.0015 -1>}
 clavicle_l {parent=spine_05, loc=<5.4831 -0.1053 -3.2946>, rot=<0.0465 -0.7289 -0.009 -0.6829>}
 clavicle_r {parent=spine_05, loc=<5.4831 -0.1053 3.2946>, rot=<0.7289 0.0465 -0.6829 0.009>}
 neck_01 {parent=spine_05, loc=<11.3138 0.1529 0>, rot=<0 0 -0.2003 -0.9797>}
 neck_02 {parent=neck_01, loc=<4.5627 0 0>, rot=<0 0 0.0504 -0.9987>}
 head {parent=neck_02, loc=<5.3914 0 0>, rot=<0 0 0.115 -0.9934>}
 faceAttach {parent=head, loc=<-3.5442 -0.4617 -0.0134>, rot=<0.0082 -0.7071 -0.0082 -0.7071>, scale=<0.89 0.89 0.89>}
 C_brow_mid {parent=faceAttach, loc=<-0.0354 10.1551 8.6623>, rot=<0 -0.0241 0.0048 -0.9997>}
 L_brow_mid {parent=faceAttach, loc=<3.5162 9.5465 10.0657>, rot=<0 -0.0032 0.2553 -0.9669>}
 L_brow_outer {parent=faceAttach, loc=<5.9541 7.5065 9.9635>, rot=<0.1173 0.1438 0.4072 -0.8943>}
 L_cheek_inner {parent=faceAttach, loc=<4.4352 8.4145 3.1576>, rot=<-0.0455 -0.2061 0.3265 -0.9213>}
 C_lip_upper_mid {parent=faceAttach, loc=<-0.023 10.2197 1.2402>, rot=<0.2225 0.0014 -0.0005 -0.9749>}
 L_lip_upper_outer {parent=faceAttach, loc=<1.3287 9.3628 1.1488>, rot=<0.0359 -0.0729 0.2627 -0.9614>}
 L_lip_corner {parent=faceAttach, loc=<2.2147 8.844 1.1198>, rot=<0.0033 0.0076 -0.0147 -0.9999>}
 C_nose_bridge {parent=faceAttach, loc=<-0.0368 11.2897 5.5674>, rot=<-0.2218 -0.0155 -0.0043 -0.975>}
 R_brow_mid {parent=faceAttach, loc=<-3.547 9.5473 10.0714>, rot=<-0.0001 0.0042 -0.2553 -0.9669>}
 R_brow_outer {parent=faceAttach, loc=<-5.9888 7.4997 9.9426>, rot=<0.0563 -0.067 -0.4278 -0.8996>}
 R_cheek_inner {parent=faceAttach, loc=<-4.4532 8.4232 3.1498>, rot=<-0.046 0.2075 -0.3264 -0.9211>}
 R_lip_corner {parent=faceAttach, loc=<-2.3008 8.7654 1.0868>, rot=<0.003 -0.0079 0.0146 -0.9999>}
 C_jaw {parent=faceAttach, loc=<-0.0407 3.0671 1.1699>, rot=<0 0 0 1>}
 L_lip_lower_outer {parent=C_jaw, loc=<1.3494 6.2216 -0.0927>, rot=<0.0022 -0.014 0.2607 -0.9653>}
 C_lip_lower_mid {parent=C_jaw, loc=<0.0258 6.923 -0.3352>, rot=<0 -0.0225 -0.0048 -0.9997>}
 R_lip_lower_outer {parent=C_jaw, loc=<-1.3495 6.4509 -0.2156>, rot=<-0.008 0.0693 -0.2166 -0.9738>}
 R_eye {parent=faceAttach, loc=<-3.2763 6.9812 7.9429>, rot=<0 0 0.9998 -0.0187>}
 R_eye_lid_upper_mid {parent=R_eye, loc=<-0.0697 -0.9844 0.7145>, rot=<0.0523 -0.0009 0.0167 -0.9985>}
 R_eye_lid_lower_mid {parent=R_eye, loc=<0.0515 -0.3971 -1.0305>, rot=<0.0329 -0.0005 0.0167 -0.9993>}
 L_eye {parent=faceAttach, loc=<3.2461 6.9812 7.9429>, rot=<0 0 0.9998 0.0187>}
 L_eye_lid_upper_mid {parent=L_eye, loc=<0.0528 -1.0387 0.7079>, rot=<0.0523 0 0 -0.9986>}
 L_eye_lid_lower_mid {parent=L_eye, loc=<0.1623 -0.4796 -1.078>, rot=<0.0329 0 0 -0.9995>}
 R_lip_upper_outer {parent=faceAttach, loc=<-1.3728 9.393 1.2024>, rot=<0.0378 0.0676 -0.2607 -0.9623>}
 dyn_bk_hair_MSTR {parent=head, loc=<7.019 -4.3723 0>, rot=<0 0 0 -1>}
 dyn_r_hair_MSTR {parent=dyn_bk_hair_MSTR, loc=<-0.0292 0.7158 4.5016>, rot=<0.1928 0.2828 0.938 0.0545>}
 dyn_r_hair_01 {parent=dyn_r_hair_MSTR, loc=<4.5312 0 0>, rot=<0.6213 0.0122 -0.281 -0.7313>}
 dyn_r_hair_02 {parent=dyn_r_hair_01, loc=<5.8416 0 0>, rot=<0.1867 0.1387 -0.1235 -0.9647>}
 dyn_r_hair_03 {parent=dyn_r_hair_02, loc=<4.4696 0 0>, rot=<0.0858 -0.1598 -0.0545 -0.9819>}
 dyn_l_hair_MSTR {parent=dyn_bk_hair_MSTR, loc=<0.1417 0.5053 -4.0162>, rot=<-0.0315 -0.9587 -0.139 -0.2461>}
 dyn_l_hair_01 {parent=dyn_l_hair_MSTR, loc=<4.6832 0 0>, rot=<-0.4559 -0.0724 0.2036 -0.8634>}
 dyn_l_hair_02 {parent=dyn_l_hair_01, loc=<5.4495 0 0>, rot=<-0.269 0.1459 0.1435 -0.9412>}
 dyn_l_hair_03 {parent=dyn_l_hair_02, loc=<4.6352 0 0>, rot=<-0.0475 -0.0532 0.0319 -0.9969>}
 dyn_bk_hair_01 {parent=dyn_bk_hair_MSTR, loc=<-4.2179 -1.2302 0>, rot=<0.1442 -0.6922 0.6922 -0.1442>}
 dyn_bk_hair_02 {parent=dyn_bk_hair_01, loc=<4.1987 0 0>, rot=<0 0.1395 0 -0.9902>}
 dyn_bk_hair_03 {parent=dyn_bk_hair_02, loc=<4.0773 0 0>, rot=<0 -0.0708 0 -0.9975>}
 dyn_bk_hair_04 {parent=dyn_bk_hair_03, loc=<4.5558 0 0>, rot=<0 -0.188 0 -0.9822>}
 dyn_bk_hair_05 {parent=dyn_bk_hair_04, loc=<4.1233 0 0>, rot=<0 -0.0171 0 -0.9999>}
 dyn_bk_l_hair_01 {parent=dyn_bk_hair_MSTR, loc=<-4.5452 -0.4788 -3.5083>, rot=<0.1831 -0.7573 0.5802 -0.2372>}
 dyn_hair_extra_01 {parent=dyn_bk_l_hair_01, loc=<1.8115 0.875 1.7989>, rot=<0.1192 0.2239 -0.1304 -0.9585>}
 dyn_bk_l_hair_02 {parent=dyn_bk_l_hair_01, loc=<3.8513 0 0>, rot=<-0.0122 0.0194 0.0087 -0.9997>}
 dyn_bk_l_hair_03 {parent=dyn_bk_l_hair_02, loc=<3.9336 0 0>, rot=<-0.0414 -0.0556 0.0274 -0.9972>}
 dyn_bk_l_hair_04 {parent=dyn_bk_l_hair_03, loc=<3.932 0 0>, rot=<-0.2145 -0.1203 0.0917 -0.9649>}
 dyn_bk_l_hair_05 {parent=dyn_bk_l_hair_04, loc=<3.8459 0 0>, rot=<0.1645 -0.0747 -0.0341 -0.983>}
 dyn_bk_r_hair_01 {parent=dyn_bk_hair_MSTR, loc=<-4.7307 -0.0529 3.7507>, rot=<0.1761 -0.5864 0.7792 -0.134>}
 dyn_bk_r_hair_02 {parent=dyn_bk_r_hair_01, loc=<3.6358 0 0>, rot=<0.0426 0.1615 -0.0292 -0.9855>}
 dyn_bk_r_hair_03 {parent=dyn_bk_r_hair_02, loc=<3.9106 0 0>, rot=<0.0212 -0.0881 -0.017 -0.9957>}
 dyn_bk_r_hair_04 {parent=dyn_bk_r_hair_03, loc=<3.799 0 0>, rot=<0.0817 -0.1236 -0.0407 -0.9881>}
 dyn_hair_extra_02 {parent=dyn_bk_r_hair_04, loc=<0.7938 2.0361 2.3296>, rot=<-0.0763 0.3856 -0.4329 -0.8113>}
 dyn_bk_r_hair_05 {parent=dyn_bk_r_hair_04, loc=<4.5528 0 0>, rot=<-0.0293 -0.1045 0.007 -0.9941>}
 dyn_l_side_hair_MSTR {parent=head, loc=<5.2926 3.0733 -6.593>, rot=<0.0284 -0.989 -0.1425 0.0286>}
 dyn_l_side_hair_01 {parent=dyn_l_side_hair_MSTR, loc=<0.269 0.0108 0.0168>, rot=<0 0 0 1>}
 dyn_l_side_hair_02 {parent=dyn_l_side_hair_01, loc=<3.1107 -0.364 -0.1232>, rot=<0.0452 0.0266 -0.0828 -0.9952>}
 dyn_r_side_hair_MSTR {parent=head, loc=<5.3362 3.0902 6.3273>, rot=<0.0233 -0.0243 0.9994 -0.0108>}
 dyn_r_side_hair_01 {parent=dyn_r_side_hair_MSTR, loc=<0.352 -0.0043 -0.0164>, rot=<0 0 0 1>}
 dyn_r_side_hair_02 {parent=dyn_r_side_hair_01, loc=<2.775 0 0>, rot=<-0.0728 -0.0315 0.0396 -0.9961>}
 dyn_glasses_strap {parent=head, loc=<4.4176 -8.2471 -3.4753>, rot=<0.6329 0.54 -0.3453 0.4343>}
]
