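# Post-process-style secondary-motion AnimBP for the female medium Street Fashion Diamond cosmetic body, copying its pose from the attached parent mesh.
# When bIsInFrontEnd is true, two RigidBody passes simulate the body and hat physics assets relative to spine_05.
# Otherwise, a Control Rig feeds spring-driven AnimDynamics on side hair and glasses strap bones, then one RigidBody pass using the body physics asset.
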AnimBlueprint /BRCosmetics/Characters/Player/Female/Medium/Bodies/F_MED_Street_Fashion_Diamond/Meshes/Parts/F_MED_Street_Fashion_Diamond_AnimBP
target_skeleton=/Game/Characters/Player/Female/Medium/Bodies/F_MED_Street_Fashion_Diamond/Meshes/Parts/F_MED_Street_Fashion_Diamond_Skeleton
graphs [
 AnimGraph {nodes [
   n1 = Control Rig!(Source=Copy Pose From Mesh(SourceMeshComponent=Get Body_Mesh()) {bUseAttachedParent=true}, bAlphaBoolEnabled:bool=True) -> (Pose) {
    alphaInputType=Bool
    alphaBoolBlend {blendOption=Linear}
    controlRigAssetReference {blueprintRigClass=/Game/Characters/Player/Female/Medium/Bodies/F_MED_Street_Fashion_Diamond/Meshes/Parts/F_MED_Street_Fashion_Diamond_CtrlRig/F_MED_Street_Fashion_Diamond_CtrlRig.F_MED_Street_Fashion_Diamond_CtrlRig_C}
   }
   n2 = Anim Dynamics - Bone: dyn_l_side_hair_01(ComponentPose=Local To Component(LocalPose=@n1), Alpha=1, ExternalForce="0,0,0") {
    angularDampingOverride=0.98
    angularSpringConstant=300
    bAngularSpring=true
    bGravityOverrideInSimSpace=true
    bOverrideAngularDamping=true
    simulationSpace=RootRelative
    alphaBoolBlend {blendOption=Linear}
    boundBone {boneName=dyn_l_side_hair_01}
    componentLinearAccScale=<1.25 1.25 1.25>
    componentLinearVelScale=<1.25 1.25 1.25>
    physicsBodyDefinitions [{
      sphereCollisionRadius=0
      boundBone {boneName=dyn_l_side_hair_01}
      boxExtents=<5 20 20>
      constraintSetup {angularLimitsMax=<2 1 25>, angularLimitsMin=<-25 -1 -10>, angularTarget=<0 0 0>}
      localJointOffset=<1.5 0 0>
     }]
   }
   n3 = Anim Dynamics - Bone: dyn_l_side_hair_02(ComponentPose=@n2, Alpha=1, ExternalForce="0,0,0") {
    angularDampingOverride=0.98
    angularSpringConstant=300
    bAngularSpring=true
    bGravityOverrideInSimSpace=true
    bOverrideAngularDamping=true
    simulationSpace=RootRelative
    alphaBoolBlend {blendOption=Linear}
    boundBone {boneName=dyn_l_side_hair_02}
    componentLinearAccScale=<1.25 1.25 1.25>
    componentLinearVelScale=<1.25 1.25 1.25>
    physicsBodyDefinitions [{
      sphereCollisionRadius=0
      boundBone {boneName=dyn_l_side_hair_02}
      boxExtents=<5 15 15>
      constraintSetup {angularLimitsMax=<2 1 25>, angularLimitsMin=<-25 -1 -10>, angularTarget=<0 0 0>}
      localJointOffset=<1.5 0 0>
     }]
   }
   n4 = Anim Dynamics - Bone: dyn_r_side_hair_01(ComponentPose=@n3, Alpha=1, ExternalForce="0,0,0") {
    angularDampingOverride=0.98
    angularSpringConstant=300
    bAngularSpring=true
    bGravityOverrideInSimSpace=true
    bOverrideAngularDamping=true
    simulationSpace=RootRelative
    alphaBoolBlend {blendOption=Linear}
    boundBone {boneName=dyn_r_side_hair_01}
    componentLinearAccScale=<1.25 1.25 1.25>
    componentLinearVelScale=<1.25 1.25 1.25>
    physicsBodyDefinitions [{
      sphereCollisionRadius=0
      boundBone {boneName=dyn_r_side_hair_01}
      boxExtents=<5 20 20>
      constraintSetup {angularLimitsMax=<2 1 15>, angularLimitsMin=<-25 -1 -20>, angularTarget=<0 0 0>}
      localJointOffset=<1.5 0 0>
     }]
   }
   n5 = Anim Dynamics - Bone: dyn_r_side_hair_02(ComponentPose=@n4, Alpha=1, ExternalForce="0,0,0") {
    angularDampingOverride=0.98
    angularSpringConstant=300
    bAngularSpring=true
    bGravityOverrideInSimSpace=true
    bOverrideAngularDamping=true
    simulationSpace=RootRelative
    alphaBoolBlend {blendOption=Linear}
    boundBone {boneName=dyn_r_side_hair_02}
    componentLinearAccScale=<1.25 1.25 1.25>
    componentLinearVelScale=<1.25 1.25 1.25>
    physicsBodyDefinitions [{
      sphereCollisionRadius=0
      boundBone {boneName=dyn_r_side_hair_02}
      boxExtents=<5 15 15>
      constraintSetup {angularLimitsMax=<2 1 25>, angularLimitsMin=<-25 -1 -25>, angularTarget=<0 0 0>}
      localJointOffset=<1.5 0 0>
     }]
   }
   n6 = Anim Dynamics - Bone: dyn_glasses_strap(ComponentPose=@n5, Alpha=1, ExternalForce="0,0,0") {
    angularDampingOverride=0.98
    angularSpringConstant=500
    bAngularSpring=true
    bGravityOverrideInSimSpace=true
    bOverrideAngularDamping=true
    simulationSpace=RootRelative
    alphaBoolBlend {blendOption=Linear}
    boundBone {boneName=dyn_glasses_strap}
    componentLinearAccScale=<1.25 1.25 1.25>
    componentLinearVelScale=<1.25 1.25 1.25>
    physicsBodyDefinitions [{
      sphereCollisionRadius=0
      boundBone {boneName=dyn_glasses_strap}
      boxExtents=<10 10 10>
      constraintSetup {angularLimitsMax=<2 10 5>, angularLimitsMin=<-2 -20 -5>, angularTarget=<0 0 0>}
      localJointOffset=<2 0 0>
     }]
   }
   n7 = RigidBody(ComponentPose=@n6) {
    bClampLinearTranslationLimitToRefPose=true
    bOverrideWorldGravity=true
    overridePhysicsAsset=&/F_MED_Street_Fashion_Diamond_new_Physics
    simulationSpace=BaseBoneSpace
    alphaBoolBlend {blendOption=Linear}
    baseBoneRef {boneName=spine_05}
    componentAppliedLinearAccClamp=<5000 3000 3000>
    componentLinearAccScale=<0.35 0.3 0.3>
    componentLinearVelScale=<0.35 0.3 0.3>
    overrideWorldGravity {z=-980}
   }
   n8 = Get Body_Mesh()
   n9 = RigidBody(ComponentPose=Local To Component(LocalPose=Copy Pose From Mesh(SourceMeshComponent=@n8) {bUseAttachedParent=true})) {
    bClampLinearTranslationLimitToRefPose=true
    bOverrideWorldGravity=true
    overridePhysicsAsset=&F_MED_Street_Fashion_Diamond_new_Physics
    simulationSpace=BaseBoneSpace
    alphaBoolBlend {blendOption=Linear}
    baseBoneRef {boneName=spine_05}
    componentAppliedLinearAccClamp=<5000 3000 3000>
    componentLinearAccScale=<0.35 0.3 0.3>
    componentLinearVelScale=<0.35 0.3 0.3>
    overrideWorldGravity {z=-980}
   }
   n10 = RigidBody(ComponentPose=@n9) {
    bClampLinearTranslationLimitToRefPose=true
    bOverrideWorldGravity=true
    overridePhysicsAsset=&/F_MED_Street_Fashion_Diamond_hat_Physics
    simulationSpace=BaseBoneSpace
    alphaBoolBlend {blendOption=Linear}
    baseBoneRef {boneName=spine_05}
    componentLinearAccScale=<0.3 0.3 0.4>
    componentLinearVelScale=<0.3 0.3 0.4>
    overrideWorldGravity {z=-980}
   }
   Output Pose(Result=Blend Poses by bool(BlendPose_0=Component To Local(ComponentPose=@n10), BlendPose_1=Component To Local(ComponentPose=@n7), bActiveValue=Get bIsInFrontEnd()) {blendType=Linear}) {layerGroup=None}
  ]}
 EventGraph {nodes [
   Event BlueprintInitializeAnimation()
   > Set Body_Mesh(Body_Mesh=GetOwnerSkeletalMeshComponent())
   Event BlueprintUpdateAnimation()
  ]}
]
variables [Body_Mesh {type=SkeletalMeshComponent}]
